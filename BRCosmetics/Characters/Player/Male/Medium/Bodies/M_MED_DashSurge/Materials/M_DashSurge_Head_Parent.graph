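# Parent material for the DashSurge male medium character's head, adding subsurface profile shading to the shared 2023 character material pipeline.
# The PreFX function's attributes have their shading model forced to Subsurface Profile and their opacity overridden before entering the PostFX function.
# The Tech Art Mask's blue channel blends base opacity toward that same channel scaled by the SubsurfaceOpacityMapMultiplier parameter.
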Material /BRCosmetics/Characters/Player/Male/Medium/Bodies/M_MED_DashSurge/Materials/M_DashSurge_Head_Parent
graph {
 properties {
  bUseMaterialAttributes=true
  bUsedWithMeshDeformer=true
  bUsedWithMorphTargets=true
  bUsedWithSkeletalMesh=true
  pixelDepthOffsetMode=PDOM_Legacy
  shadingModel=MSM_FromMaterialExpression
  usedShadingModels=Default Lit | Subsurface | Clear Coat | Subsurface Profile | Cloth
 }
 nodes [
  n1 = MaterialFunctionCall() {materialFunction=&MF_CharacterParent_2023_PreFX}
  n2 = GetMaterialAttributes(None=@n1.Result) {attributeGetTypes [#1]}
  n3 = ComponentMask(None=@n1.Tech Art Mask) {b=true}
  n4 = Multiply(A=ScalarParameter() {defaultValue=1, group=DashSurge, parameterName=SubsurfaceOpacityMapMultiplier}, B=@n3)
  MaterialFunctionCall(
   Material=SetMaterialAttributes!(
    MaterialAttributes=@n2.MaterialAttributes
    Shading Model=ShadingModel() {shadingModel=MSM_SubsurfaceProfile}
    Opacity=LinearInterpolate(A=@n2.Opacity, B=@n4, Alpha=@n3)
   ) {attributeSetTypes [#2, #1]}
   M=@n1.M
   Pre Skinned Local Position=@n1.M
   Distance Blend=@n1.M
   Fresnel=@n1.M
  ) {materialFunction=&MF_CharacterParent_2023_PostFX}
 ]
}
