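# Crouch physics asset for the DashSurge medium male body, covering dangling belt straps, shirt flaps, a leek and a soy sauce prop.
# Capsule bodies on pelvis, spine_01 and thighs accompany the dyn_ bone chains.
# Chain bases join the pelvis or belt root with free angular constraints, while child segments have swing and twist limits.
# The dyn_leek body has no constraint to its base, and spine_01 has none.
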PhysicsAsset /BRCosmetics/Characters/Player/Male/Medium/Bodies/M_MED_DashSurge/Meshes/M_MED_DashSurge_Crouch_Physics
bodies {
 dyn_belt_down_l_01 [{shapeType=Capsule, length=5, radius=2, shapeName=dyn_belt_down_l_01_capsule, center=<4 0 0>, rotation=<rot 90 0 0>}]
 dyn_belt_down_l_02 [{shapeType=Capsule, length=4.5, radius=2, shapeName=dyn_belt_down_l_02_capsule, center=<4 0 0>, rotation=<rot 90 0 0>}]
 dyn_belt_down_l_03 [{shapeType=Capsule, length=4.5602, radius=2.9398, shapeName=dyn_belt_down_l_03_capsule, center=<4 0 0>, rotation=<rot 90 0 0>}]
 dyn_belt_down_l_base [{shapeType=Capsule, length=1, radius=2, shapeName=dyn_belt_down_l_base_capsule, rotation=<rot 30 -180 -180>}]
 dyn_belt_down_r_01 [{shapeType=Capsule, length=5, radius=2, shapeName=dyn_belt_down_r_01_capsule, center=<4 0 0>, rotation=<rot 90 0 0>}]
 dyn_belt_down_r_02 [{shapeType=Capsule, length=5.6739, radius=2, shapeName=dyn_belt_down_r_02_capsule, center=<5 0 -1.7321>, rotation=<rot 90 0 0>}]
 dyn_belt_down_r_03 [{shapeType=Capsule, length=5.5623, radius=2, shapeName=dyn_belt_down_r_03_capsule, center=<5 0 0>, rotation=<rot 90 0 0>}]
 dyn_belt_down_r_base [{shapeType=Capsule, length=1, radius=2, shapeName=dyn_belt_down_r_base_capsule, rotation=<rot 29.9998 179.9999 180>}]
 dyn_belt_root [{shapeType=Capsule, length=3, radius=3, shapeName=dyn_belt_root_capsule, center=<6 0 0>, rotation=<rot 90 0 0>}]
 dyn_belt_up_l [{
   shapeType=Capsule
   length=6.1594
   radius=3.8406
   shapeName=dyn_belt_up_l_capsule
   center=<10.0543 -0.7454 -0.4795>
   rotation=<rot 5.0003 5.0004 4.9999>
  }]
 dyn_belt_up_l_base [{shapeType=Capsule, length=3, radius=3, shapeName=dyn_belt_up_l_base_capsule, center=<6 0 0>, rotation=<rot 90 0 0>}]
 dyn_belt_up_r [{
   shapeType=Capsule
   length=6.481
   radius=2.9399
   shapeName=dyn_belt_up_r_capsule
   center=<11.7949 -1.5222 0.6437>
   rotation=<rot 9.8477 -1.7546 -10.1501>
  }]
 dyn_belt_up_r_base [{shapeType=Capsule, length=3, radius=3, shapeName=dyn_belt_up_r_base_capsule, center=<6 0 0>, rotation=<rot 90 0 0>}]
 dyn_leek [{
   shapeType=Capsule
   length=52.2457
   radius=4.6023
   shapeName=dyn_leek_capsule
   center=<-32.1643 0.12 0.6174>
   rotation=<rot 0 89.8216 -89.6427>
  }]
 dyn_leek_base [{shapeType=Capsule, length=8.005, radius=0.505, shapeName=dyn_leek_base_capsule, center=<0 0 -15.6698>, rotation=<rot 0 -90 0>}]
 dyn_shirt_bk_cntr_01 [{shapeType=Capsule, length=16, radius=4, shapeName=dyn_shirt_bk_cntr_01_capsule, center=<5 0 1.5>, rotation=<rot 90 -15.4571 74.5428>}]
 dyn_shirt_bk_cntr_02 [{shapeType=Capsule, length=15, radius=5, shapeName=dyn_shirt_bk_cntr_02_capsule, center=<5 0 1.5>, rotation=<rot 90 -11.6292 78.3707>}]
 dyn_shirt_bk_cntr_base [{shapeType=Capsule, length=0.505, radius=0.505, shapeName=dyn_shirt_bk_cntr_base_capsule, rotation=<rot 0 -90 0>}]
 dyn_shirt_bk_l_01 [{shapeType=Capsule, length=8.5, radius=5, shapeName=dyn_shirt_bk_l_01_capsule, center=<6.9185 0 1.5>, rotation=<rot 90 180 -90.0001>}]
 dyn_shirt_bk_l_02 [{shapeType=Capsule, length=10.5, radius=3, shapeName=dyn_shirt_bk_l_02_capsule, center=<6.9185 0 1.5>, rotation=<rot 90 180 -90.0001>}]
 dyn_shirt_bk_l_base [{shapeType=Capsule, length=5, radius=2, shapeName=dyn_shirt_bk_l_base_capsule, rotation=<rot 0 0 90>}]
 dyn_shirt_bk_r_01 [{shapeType=Capsule, length=8.5, radius=5, shapeName=dyn_shirt_bk_l_01_capsule, center=<6.9185 0 1.5>, rotation=<rot 90 180 89.9999>}]
 dyn_shirt_bk_r_02 [{shapeType=Capsule, length=10.5, radius=3, shapeName=dyn_shirt_bk_l_02_capsule, center=<6.9185 0 1.5>, rotation=<rot 90 -165.9638 104.0361>}]
 dyn_shirt_bk_r_base [{shapeType=Capsule, length=5, radius=2, shapeName=dyn_shirt_bk_r_base_capsule, rotation=<rot 0 0 90>}]
 dyn_shirt_down_bk_l [{shapeType=Capsule, length=5, radius=3, shapeName=dyn_shirt_down_bk_l_capsule, center=<10 0 0>, rotation=<rot 0 0 90>}]
 dyn_shirt_down_bk_l_base [{shapeType=Capsule, length=1, radius=1, shapeName=dyn_shirt_down_bk_l_base_capsule, rotation=<rot 0 0 90>}]
 dyn_shirt_down_bk_r [{shapeType=Capsule, length=5, radius=3, shapeName=dyn_shirt_down_bk_r_capsule, center=<10 0 0>, rotation=<rot 0 0 90>}]
 dyn_shirt_down_bk_r_base [{shapeType=Capsule, length=1, radius=1, shapeName=dyn_shirt_down_bk_r_base_capsule, rotation=<rot 0 0 90>}]
 dyn_shirt_down_fr_l [{shapeType=Capsule, length=5, radius=3, shapeName=dyn_shirt_down_fr_l_capsule, center=<10 0 0>, rotation=<rot 0 0 90>}]
 dyn_shirt_down_fr_l_base [{shapeType=Capsule, length=1, radius=1, shapeName=dyn_shirt_down_fr_l_base_capsule, rotation=<rot 0 0 90>}]
 dyn_shirt_down_fr_r [{shapeType=Capsule, length=5, radius=3, shapeName=dyn_shirt_down_fr_r_capsule, center=<10 0 0>, rotation=<rot 0 0 90>}]
 dyn_shirt_down_fr_r_base [{shapeType=Capsule, length=1, radius=1, shapeName=dyn_shirt_down_fr_r_base_capsule, rotation=<rot 0 0 90>}]
 dyn_shirt_down_side_l [{shapeType=Capsule, length=5, radius=3, shapeName=dyn_shirt_down_side_l_capsule, center=<10 0 0>, rotation=<rot 0 0 90>}]
 dyn_shirt_down_side_l_base [{shapeType=Capsule, length=1, radius=1, shapeName=dyn_shirt_down_side_l_base_capsule, rotation=<rot 0 0 90>}]
 dyn_shirt_down_side_r [{shapeType=Capsule, length=5, radius=3, shapeName=dyn_shirt_down_side_r_capsule, center=<10 0 0>, rotation=<rot 0 0 90>}]
 dyn_shirt_down_side_r_base [{shapeType=Capsule, length=1, radius=1, shapeName=dyn_shirt_down_side_r_base_capsule, rotation=<rot 0 0 90>}]
 dyn_shirt_side_l_01 [{shapeType=Capsule, length=16, radius=8, shapeName=dyn_shirt_side_l_01_capsule, center=<11.8407 0.0001 1.5>, rotation=<rot 90 90 179.9996>}]
 dyn_shirt_side_l_02 [{shapeType=Capsule, length=15, radius=8, shapeName=dyn_shirt_side_l_02_capsule, center=<11.8407 0 1.5>, rotation=<rot 90 -165.9638 -75.9641>}]
 dyn_shirt_side_l_base [{shapeType=Capsule, length=5, radius=2, shapeName=dyn_shirt_side_l_base_capsule, rotation=<rot 0 0 90>}]
 dyn_shirt_side_r_01 [{shapeType=Capsule, length=16, radius=8, shapeName=dyn_shirt_side_l_01_capsule, center=<11.8407 0.0001 1.5>, rotation=<rot 90 90 -0.0004>}]
 dyn_shirt_side_r_02 [{shapeType=Capsule, length=15, radius=8, shapeName=dyn_shirt_side_l_02_capsule, center=<11.8407 0 1.5>, rotation=<rot 90 -165.9638 104.0359>}]
 dyn_shirt_side_r_base [{shapeType=Capsule, length=5, radius=2, shapeName=dyn_shirt_side_r_base_capsule, rotation=<rot 0 0 90>}]
 dyn_soy_sauce [{shapeType=Capsule, length=17, radius=7, shapeName=dyn_soy_sauce_capsule, center=<8.6163 0 0>, rotation=<rot 90 0 0>}]
 dyn_soy_sauce_base [{shapeType=Capsule, length=0.505, radius=5, shapeName=dyn_soy_sauce_base_capsule, rotation=<rot 90 0 0>}]
 pelvis [
  {shapeType=Capsule, length=12, radius=10}
  {shapeType=Capsule, length=10.478, radius=12.522, shapeName=pelvis_capsule, center=<-11.8186 -0.5163 0>}
 ]
 spine_01 [
  {shapeType=Capsule, length=9, radius=9, shapeName=spine_01_capsule, center=<9.3839 8.7048 0>}
  {
   shapeType=Capsule
   length=6
   radius=8.9557
   shapeName=spine_01_capsule
   center=<-0.0961 16.8907 19.0575>
   rotation=<rot -19.9207 -95.3194 91.8169>
  }
  {
   shapeType=Capsule
   length=6
   radius=8.9557
   shapeName=spine_01_capsule
   center=<12.0142 12.0076 15.5208>
   rotation=<rot -6.7177 -161.1188 138.8817>
  }
 ]
 thigh_l [
  {shapeType=Capsule, length=36, radius=10.5537, shapeName=thigh_l_capsule, center=<-24.3922 0.0002 -0.4558>, rotation=<rot 0.0001 -90.0005 90>}
  {
   shapeType=Capsule
   length=14
   radius=7
   shapeName=thigh_l_capsule
   center=<-16.2397 -0.3688 4.6037>
   rotation=<rot -81.6068 122.7319 -120.1211>
  }
 ]
 thigh_r [
  {shapeType=Capsule, length=36, radius=10.5537, shapeName=thigh_l_capsule, center=<24.3922 -0.0002 0.4558>, rotation=<rot 0.0001 -90.0005 -90>}
  {
   shapeType=Capsule
   length=1.5806
   radius=7.5
   shapeName=thigh_r_capsule
   center=<16.5604 -5.7112 1.5317>
   rotation=<rot -78.7486 24.0613 147.7671>
  }
 ]
}
constraints [
 {
  bone1Name=dyn_belt_root
  bone2Name=pelvis
  swing1LimitDegrees=45
  swing1Motion=Free
  swing2LimitDegrees=45
  swing2Motion=Free
  twistLimitDegrees=45
  twistMotion=Free
 }
 {bone1Name=dyn_belt_down_l_base, bone2Name=dyn_belt_root, swing1LimitDegrees=15, swing2LimitDegrees=15, twistLimitDegrees=5}
 {bone1Name=dyn_belt_down_l_01, bone2Name=dyn_belt_down_l_base, swing1LimitDegrees=50, swing2LimitDegrees=50, twistLimitDegrees=5}
 {bone1Name=dyn_belt_down_l_02, bone2Name=dyn_belt_down_l_01, swing1LimitDegrees=15, swing2LimitDegrees=15, twistLimitDegrees=5}
 {bone1Name=dyn_belt_down_l_03, bone2Name=dyn_belt_down_l_02, swing1LimitDegrees=15, swing2LimitDegrees=15, twistLimitDegrees=5}
 {bone1Name=dyn_belt_down_r_base, bone2Name=dyn_belt_root, swing1LimitDegrees=15, swing2LimitDegrees=15, twistLimitDegrees=5}
 {bone1Name=dyn_belt_down_r_01, bone2Name=dyn_belt_down_r_base, swing1LimitDegrees=50, swing2LimitDegrees=50, twistLimitDegrees=5}
 {bone1Name=dyn_belt_down_r_02, bone2Name=dyn_belt_down_r_01, swing1LimitDegrees=15, swing2LimitDegrees=15, twistLimitDegrees=5}
 {bone1Name=dyn_belt_down_r_03, bone2Name=dyn_belt_down_r_02, swing1LimitDegrees=15, swing2LimitDegrees=15, twistLimitDegrees=5}
 {
  bone1Name=dyn_belt_up_l_base
  bone2Name=dyn_belt_root
  swing1LimitDegrees=45
  swing1Motion=Free
  swing2LimitDegrees=45
  swing2Motion=Free
  twistLimitDegrees=45
  twistMotion=Free
 }
 {bone1Name=dyn_belt_up_l, bone2Name=dyn_belt_up_l_base, swing1LimitDegrees=20, swing2LimitDegrees=80, twistLimitDegrees=30}
 {
  bone1Name=dyn_belt_up_r_base
  bone2Name=dyn_belt_root
  swing1LimitDegrees=45
  swing1Motion=Free
  swing2LimitDegrees=45
  swing2Motion=Free
  twistLimitDegrees=45
  twistMotion=Free
 }
 {bone1Name=dyn_belt_up_r, bone2Name=dyn_belt_up_r_base, swing1LimitDegrees=40, swing2LimitDegrees=50, twistLimitDegrees=60}
 {bone1Name=dyn_soy_sauce, bone2Name=dyn_soy_sauce_base, swing1LimitDegrees=15, swing2LimitDegrees=15, twistLimitDegrees=5}
 {
  bone1Name=dyn_soy_sauce_base
  bone2Name=pelvis
  swing1LimitDegrees=45
  swing1Motion=Free
  swing2LimitDegrees=45
  swing2Motion=Free
  twistLimitDegrees=45
  twistMotion=Free
 }
 {
  bone1Name=dyn_shirt_bk_l_base
  bone2Name=pelvis
  swing1LimitDegrees=45
  swing1Motion=Free
  swing2LimitDegrees=45
  swing2Motion=Free
  twistLimitDegrees=45
  twistMotion=Free
 }
 {bone1Name=dyn_shirt_bk_l_01, bone2Name=dyn_shirt_bk_l_base, swing1LimitDegrees=15, swing2LimitDegrees=45, twistLimitDegrees=5}
 {bone1Name=dyn_shirt_bk_l_02, bone2Name=dyn_shirt_bk_l_01, swing1LimitDegrees=5, swing2LimitDegrees=45, twistLimitDegrees=5}
 {
  bone1Name=dyn_shirt_bk_r_base
  bone2Name=pelvis
  swing1LimitDegrees=45
  swing1Motion=Free
  swing2LimitDegrees=45
  swing2Motion=Free
  twistLimitDegrees=45
  twistMotion=Free
 }
 {
  bone1Name=dyn_shirt_down_bk_l_base
  bone2Name=pelvis
  swing1LimitDegrees=45
  swing1Motion=Free
  swing2LimitDegrees=45
  swing2Motion=Free
  twistLimitDegrees=45
  twistMotion=Free
 }
 {bone1Name=dyn_shirt_down_bk_l, bone2Name=dyn_shirt_down_bk_l_base, swing1LimitDegrees=10, swing2LimitDegrees=30, twistLimitDegrees=15}
 {
  bone1Name=dyn_shirt_down_bk_r_base
  bone2Name=pelvis
  swing1LimitDegrees=45
  swing1Motion=Free
  swing2LimitDegrees=45
  swing2Motion=Free
  twistLimitDegrees=45
  twistMotion=Free
 }
 {bone1Name=dyn_shirt_down_bk_r, bone2Name=dyn_shirt_down_bk_r_base, swing1LimitDegrees=10, swing2LimitDegrees=30, twistLimitDegrees=15}
 {
  bone1Name=dyn_shirt_down_fr_l_base
  bone2Name=pelvis
  swing1LimitDegrees=45
  swing1Motion=Free
  swing2LimitDegrees=45
  swing2Motion=Free
  twistLimitDegrees=45
  twistMotion=Free
 }
 {bone1Name=dyn_shirt_down_fr_l, bone2Name=dyn_shirt_down_fr_l_base, swing1LimitDegrees=10, swing2LimitDegrees=30, twistLimitDegrees=15}
 {
  bone1Name=dyn_shirt_down_fr_r_base
  bone2Name=pelvis
  swing1LimitDegrees=45
  swing1Motion=Free
  swing2LimitDegrees=45
  swing2Motion=Free
  twistLimitDegrees=45
  twistMotion=Free
 }
 {bone1Name=dyn_shirt_down_fr_r, bone2Name=dyn_shirt_down_fr_r_base, swing1LimitDegrees=10, swing2LimitDegrees=30, twistLimitDegrees=15}
 {
  bone1Name=dyn_shirt_down_side_l_base
  bone2Name=pelvis
  swing1LimitDegrees=45
  swing1Motion=Free
  swing2LimitDegrees=45
  swing2Motion=Free
  twistLimitDegrees=45
  twistMotion=Free
 }
 {bone1Name=dyn_shirt_down_side_l, bone2Name=dyn_shirt_down_side_l_base, swing1LimitDegrees=10, swing2LimitDegrees=30, twistLimitDegrees=15}
 {
  bone1Name=dyn_shirt_down_side_r_base
  bone2Name=pelvis
  swing1LimitDegrees=45
  swing1Motion=Free
  swing2LimitDegrees=45
  swing2Motion=Free
  twistLimitDegrees=45
  twistMotion=Free
 }
 {bone1Name=dyn_shirt_down_side_r, bone2Name=dyn_shirt_down_side_r_base, swing1LimitDegrees=10, swing2LimitDegrees=30, twistLimitDegrees=15}
 {
  bone1Name=dyn_shirt_side_l_base
  bone2Name=pelvis
  swing1LimitDegrees=45
  swing1Motion=Free
  swing2LimitDegrees=45
  swing2Motion=Free
  twistLimitDegrees=45
  twistMotion=Free
 }
 {bone1Name=dyn_shirt_side_l_01, bone2Name=dyn_shirt_side_l_base, swing1LimitDegrees=15, swing2LimitDegrees=45, twistLimitDegrees=5}
 {bone1Name=dyn_shirt_side_l_02, bone2Name=dyn_shirt_side_l_01, swing1LimitDegrees=15, swing2LimitDegrees=45, twistLimitDegrees=5}
 {
  bone1Name=dyn_shirt_side_r_base
  bone2Name=pelvis
  swing1LimitDegrees=45
  swing1Motion=Free
  swing2LimitDegrees=45
  swing2Motion=Free
  twistLimitDegrees=45
  twistMotion=Free
 }
 {bone1Name=dyn_shirt_bk_r_02, bone2Name=dyn_shirt_bk_r_01, swing1LimitDegrees=5, swing2LimitDegrees=45, twistLimitDegrees=5}
 {bone1Name=dyn_shirt_bk_r_01, bone2Name=dyn_shirt_bk_r_base, swing1LimitDegrees=15, swing2LimitDegrees=45, twistLimitDegrees=5}
 {bone1Name=dyn_shirt_side_r_01, bone2Name=dyn_shirt_side_r_base, swing1LimitDegrees=15, swing2LimitDegrees=45, twistLimitDegrees=5}
 {bone1Name=dyn_shirt_side_r_02, bone2Name=dyn_shirt_side_r_01, swing1LimitDegrees=15, swing2LimitDegrees=45, twistLimitDegrees=5}
 {
  bone1Name=thigh_r
  bone2Name=pelvis
  swing1LimitDegrees=45
  swing1Motion=Free
  swing2LimitDegrees=45
  swing2Motion=Free
  twistLimitDegrees=45
  twistMotion=Free
 }
 {
  bone1Name=thigh_l
  bone2Name=pelvis
  swing1LimitDegrees=45
  swing1Motion=Free
  swing2LimitDegrees=45
  swing2Motion=Free
  twistLimitDegrees=45
  twistMotion=Free
 }
 {
  bone1Name=dyn_shirt_bk_cntr_base
  bone2Name=pelvis
  swing1LimitDegrees=45
  swing1Motion=Free
  swing2LimitDegrees=45
  swing2Motion=Free
  twistLimitDegrees=45
  twistMotion=Free
 }
 {bone1Name=dyn_shirt_bk_cntr_01, bone2Name=dyn_shirt_bk_cntr_base, swing1LimitDegrees=15, swing2LimitDegrees=25, twistLimitDegrees=5}
 {bone1Name=dyn_shirt_bk_cntr_02, bone2Name=dyn_shirt_bk_cntr_01, swing1LimitDegrees=15, swing2LimitDegrees=25, twistLimitDegrees=5}
 {
  bone1Name=dyn_leek_base
  bone2Name=pelvis
  swing1LimitDegrees=45
  swing1Motion=Free
  swing2LimitDegrees=45
  swing2Motion=Free
  twistLimitDegrees=45
  twistMotion=Free
 }
]
