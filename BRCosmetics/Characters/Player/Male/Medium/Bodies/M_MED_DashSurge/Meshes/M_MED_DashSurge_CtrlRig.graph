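# Control Rig for the DashSurge medium male player cosmetic body, adding corrective motion to deform, clothing and accessory bones.
# A sequence first runs the shared DeformRig function, then spherical pose readers that weight ModifyTransforms on specific bones.
# Thighs drive shirt bases and the soy sauce prop (left only), and lower arms drive front sleeves.
# Foot straps are cross-wired: foot_l drives the right strap, whose bone drives the left.
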ControlRigBlueprint /BRCosmetics/Characters/Player/Male/Medium/Bodies/M_MED_DashSurge/Meshes/M_MED_DashSurge_CtrlRig
graphs [
 RigVMFunctionLibrary {links [], nodes []}
 RigVMFunction_Sequence_1_ContainedGraph {nodes [
   n1 = RigVMFunctionEntryNode()
   RigVMFunction_Sequence = &RigVMFunction_Sequence(ExecuteContext=@n1.ExecuteContext)
   n2 = &RigVMFunction_Sequence(ExecuteContext=@RigVMFunction_Sequence.B)
   n3 = &RigVMFunction_Sequence(ExecuteContext=@n2.B)
   n4 = &RigVMFunction_Sequence(ExecuteContext=@n3.B)
   RigVMFunctionReturnNode(A=@RigVMFunction_Sequence.A, B=@n2.A, C=@n3.A, D=@n4.A, E=@n4.B)
  ]}
 RigVMModel {nodes [
   RigUnit_BeginExecution = &RigUnit_BeginExecution()
   n1 = &RigVMFunction_Sequence(ExecuteContext=@RigUnit_BeginExecution.ExecutePin)
   FN_DeformRig_FNC = RigVMFunctionReferenceNode(
    DeformRig_ Curve="(deformRig_l_upperarm_up_alpha_72_#1=False,deformRig_r_upperarm_up_alpha_73_#2=False,deformRig_l_upperarm_fr_alpha_74_#3=False,deformRig_r_upperarm_fr_alpha_75_CB51B8 ...(truncated)"
    DeformRig_Bool="(upperarm_1_#4=True,elbow_3_#5=True,wrist_7_#6=True,groin_8_#7=True,glute_17_#8=True,knee_9_40C408324CA261 ...(truncated)"
    DeformRig_Debug=(upperarm_debug_18_#4=False,wrist_debug_20_#6=False,lat_debug_26_#9=False,groinGlute_debug_27_#7=False)
    DeformRig_Value="(upperarm_up_18_#10=(Rotation=(X=0,Y=0,Z=0,W=1),Translation=(X=0,Y=0,Z=8),Scale3D=(X=0.7,Y=1,Z=1)),upperarm_fwd_fr_19_#11=(Rotation=(X=0,Y=0,Z=0,W=1),Translation=(X=0,Y=10,Z=0),Scale3D=(X=0.5,Y=1,Z= ...(truncated)"
    ExecuteContext=@n1.A
   )
   RigVMFunction_Sequence = RigVMAggregateNode(ExecuteContext=@n1.B)
   n2 = &RigVMFunction_Sequence(ExecuteContext=@RigVMFunction_Sequence.A)
   n3 = &RigUnit_SphericalPoseReader(
    ActiveRegionScaleFactors=(PositiveWidth=3.5,NegativeWidth=0.5,PositiveHeight=3,NegativeHeight=3)
    ActiveRegionSize=0.1
    Debug=(bDrawDebug=true,DebugScale=25,DebugSegments=20,DebugThickness=0.25)
    DriverAxis=(X=-1,Y=0,Z=0)
    DriverItem=(Type=Bone,Name="thigh_l")
    FalloffRegionScaleFactors=(PositiveWidth=1,NegativeWidth=1,PositiveHeight=1,NegativeHeight=1)
    FalloffSize=0.45
    OptionalParentItem=(Type=Bone)
    RotationOffset=(X=90,Y=0,Z=0)
    ExecutePin=@n2.A
   )
   > &RigUnit_ModifyTransforms(
    ItemToModify=((Item=(Type=Bone,Name="dyn_shirt_fr_l_base"),Transform=(Rotation=(X=0.1036,Y=0.7866,Z=-0.0795,W=-0.6036),Translation=(X=0,Y=0,Z=0),Scale3D=(X=1,Y=1,Z=1))))
    WeightMinimum=0
    Weight=@n3.OutputParam
   )
   > &RigUnit_ModifyTransforms(
    ItemToModify=((Item=(Type=Bone,Name="dyn_shirt_side_l_base"),Transform=(Rotation=(X=-0.5016,Y=0.2101,Z=-0.116,W=-0.8311),Translation=(X=0,Y=0,Z=0),Scale3D=(X=1,Y=1,Z=1))))
    WeightMinimum=0
    Weight=@n3.OutputParam
   )
   > &RigUnit_ModifyTransforms(
    ItemToModify=((Item=(Type=Bone,Name="dyn_soy_sauce_base"),Transform=(Rotation=(X=0.1508,Y=-0.2924,Z=0.5555,W=-0.7636),Translation=(X=0.6111,Y=-5.5592,Z=2.1484),Scale3D=(X=1,Y=1,Z=1))))
    WeightMinimum=0
    Weight=@n3.OutputParam
   )
   n4 = &RigUnit_SphericalPoseReader(
    ActiveRegionScaleFactors=(PositiveWidth=3.5,NegativeWidth=0.5,PositiveHeight=3,NegativeHeight=3)
    ActiveRegionSize=0.1
    Debug=(bDrawDebug=true,DebugScale=25,DebugSegments=20,DebugThickness=0.25)
    DriverAxis=(X=1,Y=0,Z=0)
    DriverItem=(Type=Bone,Name="thigh_r")
    FalloffRegionScaleFactors=(PositiveWidth=1,NegativeWidth=1,PositiveHeight=1,NegativeHeight=1)
    FalloffSize=0.45
    FlipHeightScaling=false
    FlipWidthScaling=true
    OptionalParentItem=(Type=Bone)
    RotationOffset=(X=-90,Y=0,Z=0)
    ExecutePin=@n2.B
   )
   > &RigUnit_ModifyTransforms(
    ItemToModify=((Item=(Type=Bone,Name="dyn_shirt_fr_r_base"),Transform=(Rotation=(X=0,Y=-0.7071,Z=0,W=0.7071),Translation=(X=0,Y=0,Z=0),Scale3D=(X=1,Y=1,Z=1))))
    WeightMinimum=0
    Weight=@n4.OutputParam
   )
   n5 = &RigVMFunction_Sequence(ExecuteContext=@RigVMFunction_Sequence.B)
   n6 = &RigUnit_SphericalPoseReader(
    ActiveRegionScaleFactors=(PositiveWidth=3.5,NegativeWidth=0.5,PositiveHeight=3,NegativeHeight=3)
    ActiveRegionSize=0.03
    Debug=(bDrawDebug=true,DebugScale=25,DebugSegments=20,DebugThickness=0.25)
    DriverAxis=(X=-1,Y=0,Z=0)
    DriverItem=(Type=Bone,Name="lowerarm_r")
    FalloffRegionScaleFactors=(PositiveWidth=1,NegativeWidth=1,PositiveHeight=1,NegativeHeight=1)
    FalloffSize=0.45
    FlipHeightScaling=false
    FlipWidthScaling=true
    OptionalParentItem=(Type=Bone)
    RotationOffset=(X=-90,Y=0,Z=0)
    ExecutePin=@n5.A
   )
   > &RigUnit_ModifyTransforms(
    ItemToModify=((Item=(Type=Bone,Name="dyn_sleeve_front_r"),Transform=(Rotation=(X=0,Y=-0.4226,Z=0,W=0.9063),Translation=(X=0,Y=0,Z=0),Scale3D=(X=1,Y=1,Z=1))))
    WeightMinimum=0
    Weight=@n6.OutputParam
   )
   n7 = &RigUnit_SphericalPoseReader(
    ActiveRegionScaleFactors=(PositiveWidth=3.5,NegativeWidth=0.5,PositiveHeight=3,NegativeHeight=3)
    ActiveRegionSize=0.03
    Debug=(bDrawDebug=true,DebugScale=25,DebugSegments=20,DebugThickness=0.25)
    DriverAxis=(X=1,Y=0,Z=0)
    DriverItem=(Type=Bone,Name="lowerarm_l")
    FalloffRegionScaleFactors=(PositiveWidth=1,NegativeWidth=1,PositiveHeight=1,NegativeHeight=1)
    FalloffSize=0.45
    OptionalParentItem=(Type=Bone)
    RotationOffset=(X=90,Y=0,Z=0)
    ExecutePin=@n5.B
   )
   > &RigUnit_ModifyTransforms(
    ItemToModify=((Item=(Type=Bone,Name="dyn_sleeve_front_l"),Transform=(Rotation=(X=0,Y=-0.4226,Z=0,W=0.9063),Translation=(X=0,Y=0,Z=0),Scale3D=(X=1,Y=1,Z=1))))
    WeightMinimum=0
    Weight=@n7.OutputParam
   )
   n8 = &RigVMFunction_Sequence(ExecuteContext=@RigVMFunction_Sequence.D)
   n9 = &RigUnit_SphericalPoseReader(
    ActiveRegionScaleFactors=(PositiveWidth=3.5,NegativeWidth=0.5,PositiveHeight=3,NegativeHeight=3)
    ActiveRegionSize=0.03
    Debug=(bDrawDebug=true,DebugScale=25,DebugSegments=20,DebugThickness=0.25)
    DriverAxis=(X=0,Y=1,Z=0)
    DriverItem=(Type=Bone,Name="foot_l")
    FalloffRegionScaleFactors=(PositiveWidth=1,NegativeWidth=1,PositiveHeight=1,NegativeHeight=1)
    FalloffSize=0.45
    OptionalParentItem=(Type=Bone)
    RotationOffset=(X=90,Y=0,Z=24)
    ExecutePin=@n8.A
   )
   > &RigUnit_ModifyTransforms(
    ItemToModify=((Item=(Type=Bone,Name="extra_straps_foot_r"),Transform=(Rotation=(X=0,Y=-0.7071,Z=0,W=0.7071),Translation=(X=0,Y=0,Z=0),Scale3D=(X=1,Y=2,Z=1))))
    WeightMinimum=0
    Weight=@n9.OutputParam
   )
   n10 = &RigUnit_SphericalPoseReader(
    ActiveRegionScaleFactors=(PositiveWidth=3.5,NegativeWidth=0.5,PositiveHeight=3,NegativeHeight=3)
    ActiveRegionSize=0.03
    Debug=(bDrawDebug=true,DebugScale=25,DebugSegments=20,DebugThickness=0.25)
    DriverAxis=(X=1,Y=0,Z=0)
    DriverItem=(Type=Bone,Name="extra_straps_foot_r")
    FalloffRegionScaleFactors=(PositiveWidth=1,NegativeWidth=1,PositiveHeight=1,NegativeHeight=1)
    FalloffSize=0.45
    FlipHeightScaling=false
    FlipWidthScaling=true
    OptionalParentItem=(Type=Bone)
    RotationOffset=(X=-2,Y=-80,Z=-7)
    ExecutePin=@n8.B
   )
   > &RigUnit_ModifyTransforms(
    ItemToModify=((Item=(Type=Bone,Name="extra_straps_foot_l"),Transform=(Rotation=(X=0,Y=-0.7071,Z=0,W=0.7071),Translation=(X=0,Y=0,Z=0),Scale3D=(X=1,Y=2,Z=1))))
    WeightMinimum=0
    Weight=@n10.OutputParam
   )
  ]}
]
hierarchy [
 root {loc=<0 0 0>, rot=<0 0 0 1>}
 pelvis {parent=root, loc=<0 -0.7071 93.1191>, rot=<0.0127 0.707 -0.0127 -0.707>}
 spine_01 {parent=pelvis, loc=<6.7494 -0.6388 0>, rot=<0 0 0.034 -0.9994>}
 spine_02 {parent=spine_01, loc=<7.252 -0.035 0>, rot=<0 0 0.0238 -0.9997>}
 spine_03 {parent=spine_02, loc=<6.0106 0 0>, rot=<0 0 0.0149 -0.9999>}
 spine_04 {parent=spine_03, loc=<8.1023 0 0>, rot=<0 0 -0.0132 -0.9999>}
 spine_05 {parent=spine_04, loc=<10.0142 0 0>, rot=<0 0 -0.0015 -1>}
 attach_backpack {parent=spine_05, loc=<-0.2171 -13.5865 0.013>, rot=<-0.0089 -0.7062 0.0093 -0.7079>}
 neck_01 {parent=spine_05, loc=<11.3138 0.1529 0>, rot=<0 0 -0.2003 -0.9797>}
 neck_02 {parent=neck_01, loc=<4.5627 0 0>, rot=<0 0 0.0504 -0.9987>}
 head {parent=neck_02, loc=<5.3914 0 0>, rot=<0 0 0.115 -0.9934>}
 clavicle_l {parent=spine_05, loc=<4.3575 -0.1466 -2.3223>, rot=<0.0465 -0.7289 -0.009 -0.6829>}
 upperarm_l {parent=clavicle_l, loc=<15.5111 0 0>, rot=<0.0226 -0.3234 -0.0088 -0.9459>}
 lowerarm_l {parent=upperarm_l, loc=<26.6514 0 0>, rot=<0.0391 -0.0099 -0.2453 -0.9686>}
 hand_l {parent=lowerarm_l, loc=<22.3491 0 0>, rot=<-0.7081 -0.0507 0.0975 -0.6975>}
 thumb_01_l {parent=hand_l, loc=<1.3118 -0.3191 -2.459>, rot=<0.6546 -0.4042 0.0664 -0.6354>}
 thumb_02_l {parent=thumb_01_l, loc=<5.9336 -0.0448 0.0401>, rot=<-0.0037 0.0213 0.298 -0.9543>}
 thumb_03_l {parent=thumb_02_l, loc=<2.2223 0.3682 -0.0907>, rot=<0.0118 -0.013 0.0066 -0.9998>}
 index_metacarpal_l {parent=hand_l, loc=<1.6535 0.0183 -2.96>, rot=<0.0158 -0.0124 -0.0173 -0.9996>}
 index_01_l {parent=index_metacarpal_l, loc=<8.3008 0.0357 -0.0524>, rot=<-0.0044 0.0485 0.0924 -0.9945>}
 index_02_l {parent=index_01_l, loc=<3.3587 -0.3004 0.0452>, rot=<-0.002 0.0027 0.3225 -0.9466>}
 index_03_l {parent=index_02_l, loc=<2.5578 -0.123 -0.0252>, rot=<-0.002 0.0067 0.2558 -0.9667>}
 middle_metacarpal_l {parent=hand_l, loc=<2.1856 -0.1725 -0.5984>, rot=<0.006 0.0168 -0.0329 -0.9993>}
 middle_01_l {parent=middle_metacarpal_l, loc=<7.3275 -0.2447 0.0129>, rot=<-0.0208 0.0042 0.0858 -0.9961>}
 middle_02_l {parent=middle_01_l, loc=<4.0123 0.1401 0.1175>, rot=<-0.0218 0.0258 0.3512 -0.9357>}
 middle_03_l {parent=middle_02_l, loc=<2.6077 0.0534 0.1373>, rot=<0.0082 0.0188 0.1708 -0.9851>}
 ring_metacarpal_l {parent=hand_l, loc=<2.2074 -0.3625 1.3638>, rot=<-0.0284 0.0722 -0.0458 -0.9959>}
 ring_01_l {parent=ring_metacarpal_l, loc=<7.0295 -0.6197 -0.1446>, rot=<-0.0489 -0.0252 0.1892 -0.9804>}
 ring_02_l {parent=ring_01_l, loc=<3.6174 0.4647 -0.1027>, rot=<0 0 0.2806 -0.9598>}
 ring_03_l {parent=ring_02_l, loc=<2.4152 0.114 0.0199>, rot=<0 0 0.1189 -0.9929>}
 pinky_metacarpal_l {parent=hand_l, loc=<1.7616 -0.9753 3.3814>, rot=<-0.0397 0.0814 -0.069 -0.9935>}
 pinky_01_l {parent=pinky_metacarpal_l, loc=<7.1709 -0.4831 -0.1094>, rot=<-0.0654 -0.0035 0.2242 -0.9723>}
 pinky_02_l {parent=pinky_01_l, loc=<2.8274 0.0893 -0.1043>, rot=<0.0037 0.0011 0.2854 -0.9584>}
 pinky_03_l {parent=pinky_02_l, loc=<2.2046 0.0307 -0.0078>, rot=<-0.0002 0.0098 0.1445 -0.9895>}
 deform_wrist_dn_base_l {parent=hand_l, loc=<0 -2.2342 0>, rot=<0 0 0 1>}
 deform_wrist_dn_l {parent=deform_wrist_dn_base_l, loc=<0 0 0>, rot=<0 0 0 1>}
 deform_wrist_up_base_l {parent=hand_l, loc=<0 2.2682 0>, rot=<0 0 0 1>}
 deform_wrist_up_l {parent=deform_wrist_up_base_l, loc=<0 0 0>, rot=<0 0 0 1>}
 lowerarm_twist_02_l {parent=lowerarm_l, loc=<9.6199 0 0>, rot=<0 0 0 -1>}
 lowerarm_twist_01_l {parent=lowerarm_l, loc=<17.8549 0 0>, rot=<0 0 0 -1>}
 deform_elbow_in_base_l {parent=lowerarm_l, loc=<0.2112 5.259 0.8274>, rot=<0 0 0 -1>}
 deform_elbow_in_l {parent=deform_elbow_in_base_l, loc=<0 0 0>, rot=<0 0 0 1>}
 deform_elbow_out_base_l {parent=lowerarm_l, loc=<0 -4 0>, rot=<0 0 0 -1>}
 deform_elbow_out_l {parent=deform_elbow_out_base_l, loc=<0 0 0>, rot=<0 0 0 1>}
 dyn_sleeve_bottom_l {parent=lowerarm_l, loc=<2.6276 1.93 -11.4337>, rot=<0.9787 -0.1684 -0.0485 -0.1072>}
 dyn_sleeve_front_l {parent=lowerarm_l, loc=<3.3246 9.2105 -0.1802>, rot=<0.6264 -0.0979 0.1319 -0.762>}
 dyn_sleeve_back_l {parent=lowerarm_l, loc=<-1.9689 -5.984 -0.0426>, rot=<-0.6554 0.1604 0.1284 -0.7268>}
 dyn_sleeve_top_l {parent=lowerarm_l, loc=<-1.1901 0.0719 8.8086>, rot=<-0.1704 0.0989 0.1841 -0.963>}
 upperarm_twist_01_l {parent=upperarm_l, loc=<0.0041 0 0>, rot=<0 0 0 -1>}
 deform_upperarm_bk_base_l {parent=upperarm_twist_01_l, loc=<1.0813 -7.2193 -1.9453>, rot=<0 0 0 1>}
 deform_upperarm_bk_l {parent=deform_upperarm_bk_base_l, loc=<0 0 0>, rot=<0 0 0 1>}
 deform_upperarm_fr_base_l {parent=upperarm_twist_01_l, loc=<-0.3691 7.0506 -2.765>, rot=<0 0 0 1>}
 deform_upperarm_fr_l {parent=deform_upperarm_fr_base_l, loc=<0 0 0>, rot=<0 0 0 1>}
 deform_upperarm_up_base_l {parent=upperarm_twist_01_l, loc=<-3.2216 -1.1391 3.3311>, rot=<0 0 0 1>}
 deform_upperarm_up_l {parent=deform_upperarm_up_base_l, loc=<0 0 0>, rot=<0 0 0 1>}
 upperarm_twist_02_l {parent=upperarm_l, loc=<13.7941 0 0>, rot=<0 0 0 -1>}
 dyn_scarf_base {parent=clavicle_l, loc=<9.0484 -6.9326 5.8535>, rot=<0.0008 0.0336 -0.0252 -0.9991>}
 dyn_scarf_01 {parent=dyn_scarf_base, loc=<-0.0002 -1.0945 -1.153>, rot=<-0.391 -0.4449 0.5323 -0.6049>}
 dyn_scarf_02 {parent=dyn_scarf_01, loc=<10.8414 0 0>, rot=<0.031 -0.1179 -0.0188 -0.9924>}
 dyn_scarf_03 {parent=dyn_scarf_02, loc=<10.8414 0 0>, rot=<0.0127 -0.0193 -0.0041 -0.9997>}
 dyn_scarf_04 {parent=dyn_scarf_03, loc=<10.8457 -0.0435 0.1933>, rot=<0.0349 0 0 -0.9994>}
 attach_cape {parent=spine_05, loc=<8.7965 -11.8172 0.0344>, rot=<0.062 -0.7036 -0.0615 -0.7052>}
 clavicle_r {parent=spine_05, loc=<4.3575 -0.1466 2.3223>, rot=<0.7289 0.0465 -0.6829 0.009>}
 upperarm_r {parent=clavicle_r, loc=<-15.5111 0 0>, rot=<0.0226 -0.3234 -0.0088 -0.9459>}
 lowerarm_r {parent=upperarm_r, loc=<-26.6514 0 0>, rot=<0.0391 -0.0099 -0.2453 -0.9686>}
 hand_r {parent=lowerarm_r, loc=<-22.3491 0 0>, rot=<-0.7081 -0.0507 0.0975 -0.6975>}
 thumb_01_r {parent=hand_r, loc=<-1.3118 0.3191 2.459>, rot=<0.6546 -0.4042 0.0664 -0.6354>}
 thumb_02_r {parent=thumb_01_r, loc=<-5.9336 0.0448 -0.0401>, rot=<-0.0037 0.0213 0.298 -0.9543>}
 thumb_03_r {parent=thumb_02_r, loc=<-2.2223 -0.3682 0.0907>, rot=<0.0118 -0.013 0.0066 -0.9998>}
 index_metacarpal_r {parent=hand_r, loc=<-1.6535 -0.0183 2.96>, rot=<0.0158 -0.0124 -0.0173 -0.9996>}
 index_01_r {parent=index_metacarpal_r, loc=<-8.3008 -0.0357 0.0524>, rot=<-0.0044 0.0485 0.0924 -0.9945>}
 index_02_r {parent=index_01_r, loc=<-3.3587 0.3004 -0.0452>, rot=<-0.002 0.0027 0.3225 -0.9466>}
 index_03_r {parent=index_02_r, loc=<-2.5578 0.123 0.0252>, rot=<-0.002 0.0067 0.2558 -0.9667>}
 middle_metacarpal_r {parent=hand_r, loc=<-2.1856 0.1725 0.5984>, rot=<0.006 0.0168 -0.0329 -0.9993>}
 middle_01_r {parent=middle_metacarpal_r, loc=<-7.3275 0.2447 -0.0129>, rot=<-0.0208 0.0042 0.0858 -0.9961>}
 middle_02_r {parent=middle_01_r, loc=<-4.0123 -0.1401 -0.1175>, rot=<-0.0218 0.0258 0.3512 -0.9357>}
 middle_03_r {parent=middle_02_r, loc=<-2.6077 -0.0534 -0.1373>, rot=<0.0082 0.0188 0.1708 -0.9851>}
 ring_metacarpal_r {parent=hand_r, loc=<-2.2074 0.3625 -1.3638>, rot=<-0.0284 0.0722 -0.0458 -0.9959>}
 ring_01_r {parent=ring_metacarpal_r, loc=<-7.0295 0.6197 0.1446>, rot=<-0.0489 -0.0252 0.1892 -0.9804>}
 ring_02_r {parent=ring_01_r, loc=<-3.6174 -0.4647 0.1027>, rot=<0 0 0.2806 -0.9598>}
 ring_03_r {parent=ring_02_r, loc=<-2.4152 -0.114 -0.0199>, rot=<0 0 0.1189 -0.9929>}
 pinky_metacarpal_r {parent=hand_r, loc=<-1.7616 0.9753 -3.3814>, rot=<-0.0397 0.0814 -0.069 -0.9935>}
 pinky_01_r {parent=pinky_metacarpal_r, loc=<-7.1709 0.4831 0.1094>, rot=<-0.0654 -0.0035 0.2242 -0.9723>}
 pinky_02_r {parent=pinky_01_r, loc=<-2.8274 -0.0893 0.1043>, rot=<0.0037 0.0011 0.2854 -0.9584>}
 pinky_03_r {parent=pinky_02_r, loc=<-2.2046 -0.0307 0.0078>, rot=<-0.0002 0.0098 0.1445 -0.9895>}
 deform_wrist_dn_base_r {parent=hand_r, loc=<-0.0004 2.2345 -0.0001>, rot=<0 0 0 1>}
 deform_wrist_dn_r {parent=deform_wrist_dn_base_r, loc=<0 0 0>, rot=<0 0 0 1>}
 deform_wrist_up_base_r {parent=hand_r, loc=<0.0003 -2.2684 0.0001>, rot=<0 0 0 1>}
 deform_wrist_up_r {parent=deform_wrist_up_base_r, loc=<0 0 0>, rot=<0 0 0 1>}
 lowerarm_twist_02_r {parent=lowerarm_r, loc=<-9.6199 0 0>, rot=<0 0 0 -1>}
 lowerarm_twist_01_r {parent=lowerarm_r, loc=<-17.8549 0 0>, rot=<0 0 0 -1>}
 deform_elbow_in_base_r {parent=lowerarm_r, loc=<-0.2111 -5.2591 -0.8275>, rot=<0 0 0 -1>}
 deform_elbow_in_r {parent=deform_elbow_in_base_r, loc=<-0.0001 0 0.0002>, rot=<0 0 0 1>}
 deform_elbow_out_base_r {parent=lowerarm_r, loc=<0 4 -0.0001>, rot=<0 0 0 -1>}
 deform_elbow_out_r {parent=deform_elbow_out_base_r, loc=<0 0 0>, rot=<0 0 0 1>}
 dyn_sleeve_bottom_r {parent=lowerarm_r, loc=<-2.6279 -1.9299 11.434>, rot=<0.0485 -0.1072 0.9787 0.1684>}
 dyn_sleeve_back_r {parent=lowerarm_r, loc=<1.9689 5.984 0.0425>, rot=<-0.1284 -0.7268 -0.6554 -0.1604>}
 dyn_sleeve_front_r {parent=lowerarm_r, loc=<-3.3249 -9.2104 0.1806>, rot=<-0.1319 -0.762 0.6264 0.0979>}
 dyn_sleeve_top_r {parent=lowerarm_r, loc=<1.1901 -0.0719 -8.8086>, rot=<-0.1841 -0.963 -0.1704 -0.0989>}
 upperarm_twist_01_r {parent=upperarm_r, loc=<-0.0041 0 0>, rot=<0 0 0 -1>}
 deform_upperarm_bk_base_r {parent=upperarm_twist_01_r, loc=<-1.0814 7.2192 1.9456>, rot=<0 0 0 1>}
 deform_upperarm_bk_r {parent=deform_upperarm_bk_base_r, loc=<0 0 0>, rot=<0 0 0 1>}
 deform_upperarm_fr_base_r {parent=upperarm_twist_01_r, loc=<0.3689 -7.0506 2.7652>, rot=<0 0 0 1>}
 deform_upperarm_fr_r {parent=deform_upperarm_fr_base_r, loc=<0 0 0>, rot=<0 0 0 1>}
 deform_upperarm_up_base_r {parent=upperarm_twist_01_r, loc=<3.2214 1.1391 -3.3308>, rot=<0 0 0 1>}
 deform_upperarm_up_r {parent=deform_upperarm_up_base_r, loc=<0 0 0>, rot=<0 0 0 -1>}
 upperarm_twist_02_r {parent=upperarm_r, loc=<-13.7941 0 0>, rot=<0 0 0 -1>}
 deform_pec_aim_base_l {parent=spine_04, loc=<7.8728 12.6388 -2.4123>, rot=<0.0102 -0.4602 0.3979 -0.7936>}
 deform_pec_aim_l {parent=deform_pec_aim_base_l, loc=<0 0 0>, rot=<0 0 0 1>}
 deform_pec_base_l {parent=deform_pec_aim_l, loc=<10.766 0 0>, rot=<0 0 0 1>}
 deform_pec_l {parent=deform_pec_base_l, loc=<0.055 0 0>, rot=<0 0 0 1>}
 deform_pec_aim_base_r {parent=spine_04, loc=<7.8723 12.6388 2.4123>, rot=<-0.4602 -0.0102 0.7936 0.3979>}
 deform_pec_aim_r {parent=deform_pec_aim_base_r, loc=<0 0 0>, rot=<0 0 0 1>}
 deform_pec_base_r {parent=deform_pec_aim_r, loc=<-10.7662 -0.0006 -0.0007>, rot=<0 0 0 1>}
 deform_pec_r {parent=deform_pec_base_r, loc=<0 0 0>, rot=<0 0 0 1>}
 thigh_l {parent=pelvis, loc=<-2.5693 -0.0109 -7.7501>, rot=<0.0012 0.0572 0.0155 -0.9982>}
 calf_l {parent=thigh_l, loc=<-39.0484 0 0>, rot=<0.0006 -0.0161 -0.0391 -0.9991>}
 foot_l {parent=calf_l, loc=<-42.561 0.1077 0.4292>, rot=<0.0418 -0.0455 0.0343 -0.9975>}
 ball_l {parent=foot_l, loc=<-6.1599 13.8264 0.6806>, rot=<0.0001 -0.0001 -0.7063 -0.7079>}
 extra_straps_foot_l {parent=foot_l, loc=<-1.0374 6.027 -0.2234>, rot=<-0.5042 -0.4922 -0.4745 -0.5275>}
 calf_twist_02_l {parent=calf_l, loc=<-18.8064 -0.0967 0.0397>, rot=<0.0237 0.0095 -0.0035 -0.9997>}
 calf_twist_01_l {parent=calf_l, loc=<-35.813 -0.0967 0.0397>, rot=<0.0237 0.0095 -0.0035 -0.9997>}
 deform_knee_in_base_l {parent=calf_l, loc=<1.298 -5.2957 -0.0496>, rot=<0 0 0 -1>}
 deform_knee_in_l {parent=deform_knee_in_base_l, loc=<0 0 0>, rot=<0 0 0 1>}
 deform_knee_out_base_l {parent=calf_l, loc=<0 4 0>, rot=<0 0 0 -1>}
 deform_knee_out_l {parent=deform_knee_out_base_l, loc=<0 0 0>, rot=<0 0 0 1>}
 thigh_twist_01_l {parent=thigh_l, loc=<-21.6303 0 0>, rot=<0.0013 0 0 -1>}
 deform_groin_base_l {parent=thigh_l, loc=<-2.1908 8.1142 -1.7442>, rot=<0 0 0 -1>}
 deform_groin_l {parent=deform_groin_base_l, loc=<0 0 0>, rot=<0 0 0 1>}
 dyn_shirt_fr_l_trg {parent=thigh_l, loc=<-18.4965 10.3194 -1.1073>, rot=<0.0226 -0.8792 -0.4707 -0.0701>}
 thigh_r {parent=pelvis, loc=<-2.5693 -0.0109 7.7501>, rot=<0.0572 -0.0012 0.9982 0.0155>}
 calf_r {parent=thigh_r, loc=<39.0484 0 0>, rot=<0.0006 -0.0161 -0.0391 -0.9991>}
 foot_r {parent=calf_r, loc=<42.561 -0.1077 -0.4292>, rot=<0.0418 -0.0455 0.0343 -0.9975>}
 ball_r {parent=foot_r, loc=<6.1599 -13.8264 -0.6806>, rot=<0.0001 -0.0001 -0.7063 -0.7079>}
 extra_straps_foot_r {parent=foot_r, loc=<1.0378 -6.0277 0.2234>, rot=<0.0049 -0.7215 -0.6924 0.002>}
 calf_twist_02_r {parent=calf_r, loc=<18.8064 0.0967 -0.0397>, rot=<0.0237 0.0095 -0.0035 -0.9997>}
 calf_twist_01_r {parent=calf_r, loc=<35.813 0.0967 -0.0397>, rot=<0.0237 0.0095 -0.0035 -0.9997>}
 deform_knee_in_base_r {parent=calf_r, loc=<-1.298 5.2957 0.0496>, rot=<0 0 0 -1>}
 deform_knee_in_r {parent=deform_knee_in_base_r, loc=<0 0 0>, rot=<0 0 0 1>}
 deform_knee_out_base_r {parent=calf_r, loc=<0 -4 0>, rot=<0 0 0 -1>}
 deform_knee_out_r {parent=deform_knee_out_base_r, loc=<0 0 0>, rot=<0 0 0 1>}
 thigh_twist_01_r {parent=thigh_r, loc=<21.6303 0 0>, rot=<0.0013 0 0 -1>}
 deform_groin_base_r {parent=thigh_r, loc=<2.1908 -8.1142 1.7442>, rot=<0 0 0 -1>}
 deform_groin_r {parent=deform_groin_base_r, loc=<0 0 0>, rot=<0 0 0 1>}
 dyn_soy_sauce_base {parent=pelvis, loc=<9.6292 8.3008 -10.2782>, rot=<-0.1846 -0.484 -0.5997 -0.6099>}
 dyn_soy_sauce {parent=dyn_soy_sauce_base, loc=<6.4706 0 0>, rot=<-0.1955 -0.5881 0.1492 -0.7705>}
 dyn_belt_root {parent=pelvis, loc=<9.8833 11.7046 1.206>, rot=<-0.5089 -0.491 -0.491 -0.5089>}
 dyn_belt_up_l_base {parent=dyn_belt_root, loc=<0.9346 -2.2059 -0.3194>, rot=<-0.2055 -0.4384 0.3138 -0.8168>}
 dyn_belt_up_l {parent=dyn_belt_up_l_base, loc=<0.3735 0 0>, rot=<0 0 0 1>}
 dyn_belt_up_r_base {parent=dyn_belt_root, loc=<0.2369 0.9429 0.2022>, rot=<0.1415 -0.3737 -0.4072 -0.8213>}
 dyn_belt_up_r {parent=dyn_belt_up_r_base, loc=<0.5327 0 0>, rot=<0 0 0 1>}
 dyn_belt_down_l_base {parent=dyn_belt_root, loc=<-0.4824 -1.1817 -1.558>, rot=<0.0314 -0.6084 0.0255 -0.7926>}
 dyn_belt_down_l_01 {parent=dyn_belt_down_l_base, loc=<0.3211 0 0>, rot=<0 0 0 1>}
 dyn_belt_down_l_02 {parent=dyn_belt_down_l_01, loc=<7.3635 0 0>, rot=<0.0057 -0.0142 0.0179 -0.9997>}
 dyn_belt_down_l_03 {parent=dyn_belt_down_l_02, loc=<7.7956 0 0>, rot=<-0.0074 0.0013 0.0457 -0.9989>}
 dyn_belt_down_r_base {parent=dyn_belt_root, loc=<-1.0582 1.1622 -1.4813>, rot=<0.0187 -0.5967 -0.0679 -0.7994>}
 dyn_belt_down_r_01 {parent=dyn_belt_down_r_base, loc=<0.4092 -0.0386 0.0372>, rot=<0 0 0 1>}
 dyn_belt_down_r_02 {parent=dyn_belt_down_r_01, loc=<7.9365 -0.0518 0.0593>, rot=<-0.0005 -0.0155 -0.0135 -0.9998>}
 dyn_belt_down_r_03 {parent=dyn_belt_down_r_02, loc=<9.3243 0 0>, rot=<0.0435 -0.0044 -0.044 -0.9981>}
 extra_leek {parent=pelvis, loc=<10.6427 -0.9769 17.5882>, rot=<0.0864 0.0289 -0.4019 -0.9111>}
 dyn_shirt_bk_l_base {parent=pelvis, loc=<9.0723 -9.7944 -8.1537>, rot=<0.0194 -0.8111 0.5595 -0.1692>}
 dyn_shirt_bk_l_01 {parent=dyn_shirt_bk_l_base, loc=<1.3911 0 0>, rot=<0 0 0 1>}
 dyn_shirt_bk_l_02 {parent=dyn_shirt_bk_l_01, loc=<11.3588 0.0009 0.0045>, rot=<0 -0.0615 0.0124 -0.998>}
 dyn_shirt_side_l_base {parent=pelvis, loc=<8.9581 0.2511 -13.9561>, rot=<-0.0248 -0.9787 -0.0318 -0.2012>}
 dyn_shirt_side_l_01 {parent=dyn_shirt_side_l_base, loc=<1.295 0 0>, rot=<0 0 0 1>}
 dyn_shirt_side_l_02 {parent=dyn_shirt_side_l_01, loc=<11.7053 0.0106 0.0379>, rot=<0.0434 -0.0817 0.0191 -0.9955>}
 dyn_shirt_fr_l_base {parent=pelvis, loc=<9.1694 10.2536 -4.1031>, rot=<0.0177 -0.9051 -0.4059 -0.1253>}
 dyn_shirt_fr_l_01 {parent=dyn_shirt_fr_l_base, loc=<1.3482 0 0>, rot=<0 0 0 1>}
 dyn_shirt_fr_l_02 {parent=dyn_shirt_fr_l_01, loc=<12.6459 -0.0896 0.0062>, rot=<0.0693 -0.0006 0.0104 -0.9975>}
 dyn_shirt_side_r_base {parent=pelvis, loc=<8.9576 0.2511 13.9561>, rot=<0.2143 0.0314 0.9759 0.0253>}
 dyn_shirt_side_r_01 {parent=dyn_shirt_side_r_base, loc=<1.2944 0 0.0002>, rot=<0 0 0 1>}
 dyn_shirt_side_r_02 {parent=dyn_shirt_side_r_01, loc=<11.7055 -0.0106 0.0378>, rot=<-0.0434 -0.0817 -0.0191 -0.9955>}
 dyn_shirt_fr_r_base {parent=pelvis, loc=<9.1697 10.2534 4.1037>, rot=<0.1251 0.4144 0.9013 -0.0189>}
 dyn_shirt_fr_r_01 {parent=dyn_shirt_fr_r_base, loc=<1.3486 -0.0001 0>, rot=<0 0 0 1>}
 dyn_shirt_fr_r_02 {parent=dyn_shirt_fr_r_01, loc=<12.6454 0.0893 0.0062>, rot=<-0.0693 -0.0006 -0.0104 -0.9975>}
 dyn_shirt_bk_r_base {parent=pelvis, loc=<9.0728 -9.7945 8.1537>, rot=<0.1359 -0.5675 0.8103 -0.0544>}
 dyn_shirt_bk_r_01 {parent=dyn_shirt_bk_r_base, loc=<1.3915 0 -0.0002>, rot=<0 0 0 1>}
 dyn_shirt_bk_r_02 {parent=dyn_shirt_bk_r_01, loc=<11.3589 -0.0009 0.0044>, rot=<0 -0.0615 -0.0124 -0.998>}
 dyn_shirt_down_side_l_base {parent=pelvis, loc=<13.8946 0.1338 -9.848>, rot=<-0.0388 -0.6709 -0.011 -0.7405>}
 dyn_shirt_down_side_l {parent=dyn_shirt_down_side_l_base, loc=<1.0544 0 0>, rot=<0 0 0 1>}
 dyn_shirt_down_fr_l_base {parent=pelvis, loc=<13.9848 4.9125 -5.2066>, rot=<-0.4456 -0.5319 -0.406 -0.5947>}
 dyn_shirt_down_fr_l {parent=dyn_shirt_down_fr_l_base, loc=<1.0544 0 0>, rot=<0 0 0 1>}
 dyn_shirt_down_bk_l_base {parent=pelvis, loc=<13.7621 -4.4638 -6.1848>, rot=<0.4288 -0.5552 0.4255 -0.5717>}
 dyn_shirt_down_bk_l {parent=dyn_shirt_down_bk_l_base, loc=<1.0544 0 0>, rot=<0 0 0 1>}
 dyn_shirt_down_side_r_base {parent=pelvis, loc=<13.8946 0.1338 9.848>, rot=<0.7405 0.011 0.6709 0.0388>}
 dyn_shirt_down_side_r {parent=dyn_shirt_down_side_r_base, loc=<1.0543 0 -0.0004>, rot=<0 0 0 1>}
 dyn_shirt_down_fr_r_base {parent=pelvis, loc=<13.9848 4.9125 5.2066>, rot=<0.5947 0.406 0.5319 0.4456>}
 dyn_shirt_down_fr_r {parent=dyn_shirt_down_fr_r_base, loc=<1.0544 0 0.0001>, rot=<0 0 0 1>}
 dyn_shirt_down_bk_r_base {parent=pelvis, loc=<13.7618 -4.4637 6.1848>, rot=<0.5717 -0.4255 0.5552 -0.4288>}
 dyn_shirt_down_bk_r {parent=dyn_shirt_down_bk_r_base, loc=<1.0544 0 0.0002>, rot=<0 0 0 -1>}
 deform_glute_base_l {parent=pelvis, loc=<-8.8064 -8.8222 -8.4287>, rot=<0.0275 0.0501 -0.4472 -0.8926>}
 deform_glute_l {parent=deform_glute_base_l, loc=<0 0 0>, rot=<0 0 0 1>}
 deform_glute_base_r {parent=pelvis, loc=<-8.8062 -8.8222 8.4287>, rot=<0.0501 -0.0275 0.8926 -0.4472>}
 deform_glute_r {parent=deform_glute_base_r, loc=<0 0 0>, rot=<0 0 0 1>}
 gravity {parent=root, loc=<3.2737 0 0>, rot=<0 0 0 1>}
]
