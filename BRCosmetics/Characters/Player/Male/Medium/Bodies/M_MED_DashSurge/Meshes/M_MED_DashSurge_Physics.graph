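# Physics asset for the male medium DashSurge cosmetic body, covering its dangling belt, shirt panels, leek and soy sauce bones.
# All bodies are capsules, including pelvis and thigh volumes around which the accessories hang.
# Most chains start at base bodies with free constraints to the pelvis or belt root, then use limited swing and twist links.
# The front shirt chain exists only on the right and joins the pelvis directly with limits.
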PhysicsAsset /BRCosmetics/Characters/Player/Male/Medium/Bodies/M_MED_DashSurge/Meshes/M_MED_DashSurge_Physics
bodies {
 dyn_belt_down_l_01 [{shapeType=Capsule, length=4, radius=1.5, shapeName=dyn_belt_down_l_01_capsule, center=<4 0 0>, rotation=<rot 90 0 0>}]
 dyn_belt_down_l_02 [{shapeType=Capsule, length=4, radius=1.5, shapeName=dyn_belt_down_l_02_capsule, center=<4 0 0>, rotation=<rot 90 0 0>}]
 dyn_belt_down_l_03 [{shapeType=Capsule, length=4, radius=2, shapeName=dyn_belt_down_l_03_capsule, center=<4 0 0>, rotation=<rot 90 0 0>}]
 dyn_belt_down_l_base [{shapeType=Capsule, length=3, radius=3, shapeName=dyn_belt_down_l_base_capsule, center=<6 0 0>, rotation=<rot 90 0 0>}]
 dyn_belt_down_r_01 [{shapeType=Capsule, length=4, radius=1.5, shapeName=dyn_belt_down_r_01_capsule, center=<4 0 0>, rotation=<rot 90 0 0>}]
 dyn_belt_down_r_02 [{shapeType=Capsule, length=5, radius=1.5, shapeName=dyn_belt_down_r_02_capsule, center=<5 0 0>, rotation=<rot 90 0 0>}]
 dyn_belt_down_r_03 [{shapeType=Capsule, length=5, radius=1.5, shapeName=dyn_belt_down_r_03_capsule, center=<5 0 0>, rotation=<rot 90 0 0>}]
 dyn_belt_down_r_base [{shapeType=Capsule, length=3, radius=3, shapeName=dyn_belt_down_r_base_capsule, center=<6 0 0>, rotation=<rot 90 0 0>}]
 dyn_belt_root [{shapeType=Capsule, length=3, radius=3, shapeName=dyn_belt_root_capsule, center=<6 0 0>, rotation=<rot 90 0 0>}]
 dyn_belt_up_l [{shapeType=Capsule, length=8, radius=4, shapeName=dyn_belt_up_l_capsule, center=<8 0 0>, rotation=<rot 90 0 0>}]
 dyn_belt_up_l_base [{shapeType=Capsule, length=3, radius=3, shapeName=dyn_belt_up_l_base_capsule, center=<6 0 0>, rotation=<rot 90 0 0>}]
 dyn_belt_up_r [{shapeType=Capsule, length=8, radius=4, shapeName=dyn_belt_up_r_capsule, center=<8 0 0>, rotation=<rot 90 0 0>}]
 dyn_belt_up_r_base [{shapeType=Capsule, length=3, radius=3, shapeName=dyn_belt_up_r_base_capsule, center=<6 0 0>, rotation=<rot 90 0 0>}]
 dyn_leek [{
   shapeType=Capsule
   length=52.2457
   radius=4.6023
   shapeName=dyn_leek_capsule
   center=<-32.1643 0.12 0.6174>
   rotation=<rot 0 89.8216 -89.6427>
  }]
 dyn_leek_base [{shapeType=Capsule, length=8.005, radius=0.505, shapeName=dyn_leek_base_capsule, center=<0 0 -15.6698>, rotation=<rot 0 -90 0>}]
 dyn_shirt_bk_cntr_01 [{shapeType=Capsule, length=16, radius=4, shapeName=dyn_shirt_bk_cntr_01_capsule, center=<5 0 1.5>, rotation=<rot 90 -15.4571 74.5428>}]
 dyn_shirt_bk_cntr_02 [{shapeType=Capsule, length=15, radius=5, shapeName=dyn_shirt_bk_cntr_02_capsule, center=<5 0 1.5>, rotation=<rot 90 -11.6292 78.3707>}]
 dyn_shirt_bk_cntr_base [{shapeType=Capsule, length=0.505, radius=0.505, shapeName=dyn_shirt_bk_cntr_base_capsule, rotation=<rot 0 -90 0>}]
 dyn_shirt_bk_l_01 [{shapeType=Capsule, length=16, radius=4, shapeName=dyn_shirt_bk_l_01_capsule, center=<5 0 1.5>, rotation=<rot 90 -15.4571 74.5428>}]
 dyn_shirt_bk_l_02 [{shapeType=Capsule, length=15, radius=5, shapeName=dyn_shirt_bk_l_02_capsule, center=<5 0 1.5>, rotation=<rot 90 -11.6292 78.3707>}]
 dyn_shirt_bk_l_base [{shapeType=Capsule, length=5, radius=2, shapeName=dyn_shirt_bk_l_base_capsule, rotation=<rot 0 0 90>}]
 dyn_shirt_bk_r_01 [{shapeType=Capsule, length=16, radius=4, shapeName=dyn_shirt_bk_l_01_capsule, center=<5 0 1.5>, rotation=<rot 90 -26.5651 -116.5651>}]
 dyn_shirt_bk_r_02 [{shapeType=Capsule, length=15, radius=5, shapeName=dyn_shirt_bk_l_02_capsule, center=<5 0 1.5>, rotation=<rot 90 90 -0.0001>}]
 dyn_shirt_bk_r_base [{shapeType=Capsule, length=5, radius=2, shapeName=dyn_shirt_bk_r_base_capsule, rotation=<rot 0 0 90>}]
 dyn_shirt_down_bk_l [{shapeType=Capsule, length=5, radius=3, shapeName=dyn_shirt_down_bk_l_capsule, center=<10 0 0>, rotation=<rot 0 0 90>}]
 dyn_shirt_down_bk_l_base [{shapeType=Capsule, length=1, radius=1, shapeName=dyn_shirt_down_bk_l_base_capsule, rotation=<rot 0 0 90>}]
 dyn_shirt_down_bk_r [{shapeType=Capsule, length=5, radius=3, shapeName=dyn_shirt_down_bk_r_capsule, center=<10 0 0>, rotation=<rot 0 0 90>}]
 dyn_shirt_down_bk_r_base [{shapeType=Capsule, length=1, radius=1, shapeName=dyn_shirt_down_bk_r_base_capsule, rotation=<rot 0 0 90>}]
 dyn_shirt_down_fr_l [{shapeType=Capsule, length=5, radius=3, shapeName=dyn_shirt_down_fr_l_capsule, center=<10 0 0>, rotation=<rot 0 0 90>}]
 dyn_shirt_down_fr_l_base [{shapeType=Capsule, length=1, radius=1, shapeName=dyn_shirt_down_fr_l_base_capsule, rotation=<rot 0 0 90>}]
 dyn_shirt_down_fr_r [{shapeType=Capsule, length=5, radius=3, shapeName=dyn_shirt_down_fr_r_capsule, center=<10 0 0>, rotation=<rot 0 0 90>}]
 dyn_shirt_down_fr_r_base [{shapeType=Capsule, length=1, radius=1, shapeName=dyn_shirt_down_fr_r_base_capsule, rotation=<rot 0 0 90>}]
 dyn_shirt_down_side_l [{shapeType=Capsule, length=5, radius=3, shapeName=dyn_shirt_down_side_l_capsule, center=<10 0 0>, rotation=<rot 0 0 90>}]
 dyn_shirt_down_side_l_base [{shapeType=Capsule, length=1, radius=1, shapeName=dyn_shirt_down_side_l_base_capsule, rotation=<rot 0 0 90>}]
 dyn_shirt_down_side_r [{shapeType=Capsule, length=5, radius=3, shapeName=dyn_shirt_down_side_r_capsule, center=<10 0 0>, rotation=<rot 0 0 90>}]
 dyn_shirt_down_side_r_base [{shapeType=Capsule, length=1, radius=1, shapeName=dyn_shirt_down_side_r_base_capsule, rotation=<rot 0 0 90>}]
 dyn_shirt_fr_r_01 [{
   shapeType=Capsule
   length=26.5
   radius=6
   shapeName=dyn_shirt_fr_l_01_capsule
   center=<13.529 19.8464 -6.7883>
   rotation=<rot 0.0006 0.0004 169.9999>
  }]
 dyn_shirt_fr_r_02 [{
   shapeType=Capsule
   length=30.5
   radius=6
   shapeName=dyn_shirt_fr_l_02_capsule
   center=<14.3346 16.6435 -15.3577>
   rotation=<rot 0.0011 0.0007 179.9997>
  }]
 dyn_shirt_side_l_01 [{
   shapeType=Capsule
   length=15
   radius=4.5
   shapeName=dyn_shirt_side_l_01_capsule
   center=<9.4912 0 2.4724>
   rotation=<rot 90 175.2364 -94.7638>
  }]
 dyn_shirt_side_l_02 [
  {shapeType=Capsule, length=15, radius=4.5, shapeName=dyn_shirt_side_l_02_capsule, center=<6.7132 0 1.3656>, rotation=<rot 90 180 89.9999>}
  {
   shapeType=Capsule
   length=15
   radius=4.5
   shapeName=dyn_shirt_side_l_02_capsule
   center=<6.545 -10.8525 -6.5486>
   rotation=<rot -14.9999 90.0003 0.0001>
  }
 ]
 dyn_shirt_side_l_base [{shapeType=Capsule, length=5, radius=2, shapeName=dyn_shirt_side_l_base_capsule, rotation=<rot 0 0 90>}]
 dyn_shirt_side_r_01 [{shapeType=Capsule, length=15, radius=4.5, shapeName=dyn_shirt_side_l_01_capsule, center=<9.4912 0 2.4724>, rotation=<rot 90 180 89.9998>}]
 dyn_shirt_side_r_02 [
  {shapeType=Capsule, length=15, radius=4.5, shapeName=dyn_shirt_side_l_02_capsule, center=<6.7132 0 1.3656>, rotation=<rot 90 180 89.9999>}
  {
   shapeType=Capsule
   length=15
   radius=4.5
   shapeName=dyn_shirt_side_l_02_capsule
   center=<6.545 9.3791 -4.8243>
   rotation=<rot -34.9998 -90.0003 -0.0001>
  }
 ]
 dyn_shirt_side_r_base [{shapeType=Capsule, length=5, radius=2, shapeName=dyn_shirt_side_r_base_capsule, rotation=<rot 0 0 90>}]
 dyn_soy_sauce [{shapeType=Capsule, length=17, radius=7, shapeName=dyn_soy_sauce_capsule, center=<8.6163 0 0>, rotation=<rot 90 0 0>}]
 dyn_soy_sauce_base [{shapeType=Capsule, length=0.505, radius=5, shapeName=dyn_soy_sauce_base_capsule, rotation=<rot 90 0 0>}]
 pelvis [
  {shapeType=Capsule, length=12, radius=10}
  {shapeType=Capsule, length=10, radius=13, shapeName=pelvis_capsule, center=<-11.8186 0 0>}
  {shapeType=Capsule, length=6, radius=13, shapeName=pelvis_capsule, center=<-13.4722 3.9748 0>}
 ]
 thigh_l [
  {
   shapeType=Capsule
   length=33
   radius=10.5537
   shapeName=thigh_l_capsule
   center=<-16.9139 -0.9547 -0.4558>
   rotation=<rot 0.0001 -90.0005 90>
  }
  {shapeType=Capsule, length=10, radius=9.5537, shapeName=thigh_r_capsule, center=<-9.073 4.0452 6.0253>, rotation=<rot 0.0002 -95.0013 80>}
 ]
 thigh_r [
  {shapeType=Capsule, length=33, radius=10.5537, shapeName=thigh_l_capsule, center=<16.9139 0.9547 0.4558>, rotation=<rot 0.0001 -90.0005 -90>}
  {shapeType=Capsule, length=10, radius=9.5537, shapeName=thigh_r_capsule, center=<10.2054 -4.6084 2.1514>, rotation=<rot 0.0001 -95.0011 -100>}
 ]
}
constraints [
 {
  bone1Name=dyn_belt_root
  bone2Name=pelvis
  swing1LimitDegrees=45
  swing1Motion=Free
  swing2LimitDegrees=45
  swing2Motion=Free
  twistLimitDegrees=45
  twistMotion=Free
 }
 {
  bone1Name=dyn_belt_down_l_base
  bone2Name=dyn_belt_root
  swing1LimitDegrees=45
  swing1Motion=Free
  swing2LimitDegrees=45
  swing2Motion=Free
  twistLimitDegrees=45
  twistMotion=Free
 }
 {bone1Name=dyn_belt_down_l_01, bone2Name=dyn_belt_down_l_base, swing1LimitDegrees=15, swing2LimitDegrees=15, twistLimitDegrees=1}
 {bone1Name=dyn_belt_down_l_02, bone2Name=dyn_belt_down_l_01, swing1LimitDegrees=15, swing2LimitDegrees=15, twistLimitDegrees=1}
 {bone1Name=dyn_belt_down_l_03, bone2Name=dyn_belt_down_l_02, swing1LimitDegrees=15, swing2LimitDegrees=15, twistLimitDegrees=1}
 {
  bone1Name=dyn_belt_down_r_base
  bone2Name=dyn_belt_root
  swing1LimitDegrees=45
  swing1Motion=Free
  swing2LimitDegrees=45
  swing2Motion=Free
  twistLimitDegrees=45
  twistMotion=Free
 }
 {bone1Name=dyn_belt_down_r_01, bone2Name=dyn_belt_down_r_base, swing1LimitDegrees=15, swing2LimitDegrees=15, twistLimitDegrees=1}
 {bone1Name=dyn_belt_down_r_02, bone2Name=dyn_belt_down_r_01, swing1LimitDegrees=15, swing2LimitDegrees=15, twistLimitDegrees=1}
 {bone1Name=dyn_belt_down_r_03, bone2Name=dyn_belt_down_r_02, swing1LimitDegrees=15, swing2LimitDegrees=15, twistLimitDegrees=1}
 {
  bone1Name=dyn_belt_up_l_base
  bone2Name=dyn_belt_root
  swing1LimitDegrees=45
  swing1Motion=Free
  swing2LimitDegrees=45
  swing2Motion=Free
  twistLimitDegrees=45
  twistMotion=Free
 }
 {bone1Name=dyn_belt_up_l, bone2Name=dyn_belt_up_l_base, swing1LimitDegrees=15, swing2LimitDegrees=15, twistLimitDegrees=1}
 {
  bone1Name=dyn_belt_up_r_base
  bone2Name=dyn_belt_root
  swing1LimitDegrees=45
  swing1Motion=Free
  swing2LimitDegrees=45
  swing2Motion=Free
  twistLimitDegrees=45
  twistMotion=Free
 }
 {bone1Name=dyn_belt_up_r, bone2Name=dyn_belt_up_r_base, swing1LimitDegrees=15, swing2LimitDegrees=15, twistLimitDegrees=1}
 {bone1Name=dyn_soy_sauce, bone2Name=dyn_soy_sauce_base, swing1LimitDegrees=15, swing2LimitDegrees=15, twistLimitDegrees=5}
 {
  bone1Name=dyn_soy_sauce_base
  bone2Name=pelvis
  swing1LimitDegrees=45
  swing1Motion=Free
  swing2LimitDegrees=45
  swing2Motion=Free
  twistLimitDegrees=45
  twistMotion=Free
 }
 {
  bone1Name=dyn_shirt_bk_l_base
  bone2Name=pelvis
  swing1LimitDegrees=45
  swing1Motion=Free
  swing2LimitDegrees=45
  swing2Motion=Free
  twistLimitDegrees=45
  twistMotion=Free
 }
 {bone1Name=dyn_shirt_bk_l_01, bone2Name=dyn_shirt_bk_l_base, swing1LimitDegrees=15, swing2LimitDegrees=30, twistLimitDegrees=5}
 {bone1Name=dyn_shirt_bk_l_02, bone2Name=dyn_shirt_bk_l_01, swing1LimitDegrees=15, swing2LimitDegrees=30, twistLimitDegrees=5}
 {
  bone1Name=dyn_shirt_bk_r_base
  bone2Name=pelvis
  swing1LimitDegrees=45
  swing1Motion=Free
  swing2LimitDegrees=45
  swing2Motion=Free
  twistLimitDegrees=45
  twistMotion=Free
 }
 {
  bone1Name=dyn_shirt_down_bk_l_base
  bone2Name=pelvis
  swing1LimitDegrees=45
  swing1Motion=Free
  swing2LimitDegrees=45
  swing2Motion=Free
  twistLimitDegrees=45
  twistMotion=Free
 }
 {bone1Name=dyn_shirt_down_bk_l, bone2Name=dyn_shirt_down_bk_l_base, swing1LimitDegrees=10, swing2LimitDegrees=30, twistLimitDegrees=15}
 {
  bone1Name=dyn_shirt_down_bk_r_base
  bone2Name=pelvis
  swing1LimitDegrees=45
  swing1Motion=Free
  swing2LimitDegrees=45
  swing2Motion=Free
  twistLimitDegrees=45
  twistMotion=Free
 }
 {bone1Name=dyn_shirt_down_bk_r, bone2Name=dyn_shirt_down_bk_r_base, swing1LimitDegrees=10, swing2LimitDegrees=30, twistLimitDegrees=15}
 {
  bone1Name=dyn_shirt_down_fr_l_base
  bone2Name=pelvis
  swing1LimitDegrees=45
  swing1Motion=Free
  swing2LimitDegrees=45
  swing2Motion=Free
  twistLimitDegrees=45
  twistMotion=Free
 }
 {bone1Name=dyn_shirt_down_fr_l, bone2Name=dyn_shirt_down_fr_l_base, swing1LimitDegrees=10, swing2LimitDegrees=30, twistLimitDegrees=15}
 {
  bone1Name=dyn_shirt_down_fr_r_base
  bone2Name=pelvis
  swing1LimitDegrees=45
  swing1Motion=Free
  swing2LimitDegrees=45
  swing2Motion=Free
  twistLimitDegrees=45
  twistMotion=Free
 }
 {bone1Name=dyn_shirt_down_fr_r, bone2Name=dyn_shirt_down_fr_r_base, swing1LimitDegrees=10, swing2LimitDegrees=30, twistLimitDegrees=15}
 {
  bone1Name=dyn_shirt_down_side_l_base
  bone2Name=pelvis
  swing1LimitDegrees=45
  swing1Motion=Free
  swing2LimitDegrees=45
  swing2Motion=Free
  twistLimitDegrees=45
  twistMotion=Free
 }
 {bone1Name=dyn_shirt_down_side_l, bone2Name=dyn_shirt_down_side_l_base, swing1LimitDegrees=10, swing2LimitDegrees=30, twistLimitDegrees=15}
 {
  bone1Name=dyn_shirt_down_side_r_base
  bone2Name=pelvis
  swing1LimitDegrees=45
  swing1Motion=Free
  swing2LimitDegrees=45
  swing2Motion=Free
  twistLimitDegrees=45
  twistMotion=Free
 }
 {bone1Name=dyn_shirt_down_side_r, bone2Name=dyn_shirt_down_side_r_base, swing1LimitDegrees=10, swing2LimitDegrees=30, twistLimitDegrees=15}
 {
  bone1Name=dyn_shirt_side_l_base
  bone2Name=pelvis
  swing1LimitDegrees=45
  swing1Motion=Free
  swing2LimitDegrees=45
  swing2Motion=Free
  twistLimitDegrees=45
  twistMotion=Free
 }
 {bone1Name=dyn_shirt_side_l_01, bone2Name=dyn_shirt_side_l_base, swing1LimitDegrees=15, swing2LimitDegrees=45, twistLimitDegrees=5}
 {
  bone1Name=dyn_shirt_side_r_base
  bone2Name=pelvis
  swing1LimitDegrees=45
  swing1Motion=Free
  swing2LimitDegrees=45
  swing2Motion=Free
  twistLimitDegrees=45
  twistMotion=Free
 }
 {bone1Name=dyn_shirt_fr_r_02, bone2Name=dyn_shirt_fr_r_01, swing1LimitDegrees=10, swing2LimitDegrees=80, twistLimitDegrees=10}
 {bone1Name=dyn_shirt_fr_r_01, bone2Name=pelvis, swing1LimitDegrees=15, swing2LimitDegrees=30, twistLimitDegrees=15}
 {bone1Name=dyn_shirt_side_r_02, bone2Name=dyn_shirt_side_r_01, swing1LimitDegrees=15, swing2LimitDegrees=45, twistLimitDegrees=5}
 {bone1Name=dyn_shirt_side_r_01, bone2Name=dyn_shirt_side_r_base, swing1LimitDegrees=15, swing2LimitDegrees=45, twistLimitDegrees=5}
 {bone1Name=dyn_shirt_bk_r_02, bone2Name=dyn_shirt_bk_r_01, swing1LimitDegrees=15, swing2LimitDegrees=25, twistLimitDegrees=5}
 {bone1Name=dyn_shirt_bk_r_01, bone2Name=dyn_shirt_bk_r_base, swing1LimitDegrees=15, swing2LimitDegrees=25, twistLimitDegrees=5}
 {bone1Name=dyn_shirt_side_l_02, bone2Name=dyn_shirt_side_l_01, swing1LimitDegrees=15, swing2LimitDegrees=45, twistLimitDegrees=5}
 {
  bone1Name=thigh_l
  bone2Name=pelvis
  swing1LimitDegrees=45
  swing1Motion=Free
  swing2LimitDegrees=45
  swing2Motion=Free
  twistLimitDegrees=45
  twistMotion=Free
 }
 {
  bone1Name=thigh_r
  bone2Name=pelvis
  swing1LimitDegrees=45
  swing1Motion=Free
  swing2LimitDegrees=45
  swing2Motion=Free
  twistLimitDegrees=45
  twistMotion=Free
 }
 {
  bone1Name=dyn_shirt_bk_cntr_base
  bone2Name=pelvis
  swing1LimitDegrees=45
  swing1Motion=Free
  swing2LimitDegrees=45
  swing2Motion=Free
  twistLimitDegrees=45
  twistMotion=Free
 }
 {bone1Name=dyn_shirt_bk_cntr_01, bone2Name=dyn_shirt_bk_cntr_base, swing1LimitDegrees=15, swing2LimitDegrees=25, twistLimitDegrees=5}
 {bone1Name=dyn_shirt_bk_cntr_02, bone2Name=dyn_shirt_bk_cntr_01, swing1LimitDegrees=15, swing2LimitDegrees=25, twistLimitDegrees=5}
 {
  bone1Name=dyn_leek_base
  bone2Name=pelvis
  swing1LimitDegrees=45
  swing1Motion=Free
  swing2LimitDegrees=45
  swing2Motion=Free
  twistLimitDegrees=45
  twistMotion=Free
 }
 {bone1Name=dyn_leek, bone2Name=dyn_leek_base, swing1LimitDegrees=30, swing2LimitDegrees=30, twistLimitDegrees=5}
]
