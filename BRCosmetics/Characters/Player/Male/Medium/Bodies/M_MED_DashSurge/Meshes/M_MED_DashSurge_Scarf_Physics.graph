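# Physics asset for the scarf on the DashSurge medium male player body cosmetic, letting the scarf move against the character's body.
# A five-capsule dyn_scarf chain runs from base to tip, with tight swing and twist limits that generally narrow along it.
# Capsules on pelvis, spine, both upper arms and right clavicle approximate the torso.
# Body joints and the scarf base's links to spine_05 and pelvis leave angular motion free.
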PhysicsAsset /BRCosmetics/Characters/Player/Male/Medium/Bodies/M_MED_DashSurge/Meshes/M_MED_DashSurge_Scarf_Physics
bodies {
 clavicle_r [{shapeType=Capsule, length=9.7191, radius=6.0296, shapeName=clavicle_l_capsule, center=<-10.4743 0 0>, rotation=<rot -90 0.0197 179.9801>}]
 dyn_scarf_01 [{shapeType=Capsule, length=3, radius=4, shapeName=dyn_scarf_01_capsule, center=<5 0 0>, rotation=<rot 79.9999 0 0>}]
 dyn_scarf_02 [{shapeType=Capsule, length=3, radius=4, shapeName=dyn_scarf_02_capsule, center=<5 0 0>, rotation=<rot 90 0 0>}]
 dyn_scarf_03 [{shapeType=Capsule, length=3, radius=3, shapeName=dyn_scarf_03_capsule, center=<5 0 0>, rotation=<rot 90 0 0>}]
 dyn_scarf_04 [{shapeType=Capsule, length=3, radius=3, shapeName=dyn_scarf_04_capsule, center=<5 0 0>, rotation=<rot 90 0 0>}]
 dyn_scarf_base [{shapeType=Capsule, length=2, radius=2, shapeName=dyn_scarf_base_capsule, rotation=<rot 90 0 0>}]
 pelvis [
  {
   shapeType=Capsule
   length=18.8292
   radius=15.5224
   shapeName=pelvis_capsule
   center=<0.8949 -0.6096 0.4165>
   rotation=<rot -7.3204 -90 -1.3552>
  }
  {shapeType=Capsule, length=2, radius=2, shapeName=dyn_scarf_base_capsule, rotation=<rot 90 0 0>}
 ]
 spine_01 [{shapeType=Capsule, length=16.8035, radius=12.3979, center=<7.6511 -1.783 0.0942>, rotation=<rot -2.9397 -90 0.0929>}]
 spine_02 [{
   shapeType=Capsule
   length=17.8843
   radius=14.5104
   shapeName=spine_02_capsule
   center=<2.6004 2.1603 0.4551>
   rotation=<rot -3.147 -90 0.9305>
  }]
 spine_03 [{shapeType=Capsule, length=16.3891, radius=12.2138, center=<1.8961 -1.2528 0.0582>, rotation=<rot -2.3684 -90 0.2751>}]
 spine_04 [{shapeType=Capsule, length=16.2983, radius=12.0624, center=<1.6804 -0.6542 -0.1104>, rotation=<rot 2.0309 -90 -0.3755>}]
 spine_05 [
  {shapeType=Capsule, length=25.9246, radius=8.3035, center=<-0.6197 -1.5207 -0.2653>, rotation=<rot -2.4753 -90 -2.0518>}
  {shapeType=Capsule, length=3, radius=9.5, shapeName=spine_05_capsule, center=<7.4041 -1.5103 -24.2949>}
  {shapeType=Capsule, length=7.5, radius=6, shapeName=spine_05_capsule, center=<-2.279 -8.4843 -21.2885>, rotation=<rot -90 90 -90.0003>}
 ]
 upperarm_l [{shapeType=Capsule, length=24.2561, radius=9.7638, center=<31.0781 2.5314 0.551>, rotation=<rot 0 101.2207 -86.8193>}]
 upperarm_r [{shapeType=Capsule, length=25.9898, radius=7.4507, center=<-13.5419 0.4597 0.9725>, rotation=<rot 0 89.3516 -88.5205>}]
}
constraints [
 {
  bone1Name=dyn_scarf_base
  bone2Name=spine_05
  swing1LimitDegrees=45
  swing1Motion=Free
  swing2LimitDegrees=45
  swing2Motion=Free
  twistLimitDegrees=45
  twistMotion=Free
 }
 {bone1Name=dyn_scarf_01, bone2Name=dyn_scarf_base, swing1LimitDegrees=30, swing2LimitDegrees=20, twistLimitDegrees=5}
 {bone1Name=dyn_scarf_02, bone2Name=dyn_scarf_01, swing1LimitDegrees=10, swing2LimitDegrees=20, twistLimitDegrees=5}
 {bone1Name=dyn_scarf_03, bone2Name=dyn_scarf_02, swing1LimitDegrees=10, swing2LimitDegrees=15, twistLimitDegrees=5}
 {bone1Name=dyn_scarf_04, bone2Name=dyn_scarf_03, swing1LimitDegrees=5, swing2LimitDegrees=15, twistLimitDegrees=5}
 {
  bone1Name=spine_05
  bone2Name=spine_04
  swing1LimitDegrees=45
  swing1Motion=Free
  swing2LimitDegrees=45
  swing2Motion=Free
  twistLimitDegrees=45
  twistMotion=Free
 }
 {
  bone1Name=spine_01
  bone2Name=pelvis
  swing1LimitDegrees=45
  swing1Motion=Free
  swing2LimitDegrees=45
  swing2Motion=Free
  twistLimitDegrees=45
  twistMotion=Free
 }
 {
  bone1Name=spine_02
  bone2Name=spine_01
  swing1LimitDegrees=45
  swing1Motion=Free
  swing2LimitDegrees=45
  swing2Motion=Free
  twistLimitDegrees=45
  twistMotion=Free
 }
 {
  bone1Name=spine_03
  bone2Name=spine_02
  swing1LimitDegrees=45
  swing1Motion=Free
  swing2LimitDegrees=45
  swing2Motion=Free
  twistLimitDegrees=45
  twistMotion=Free
 }
 {
  bone1Name=spine_04
  bone2Name=spine_03
  swing1LimitDegrees=45
  swing1Motion=Free
  swing2LimitDegrees=45
  swing2Motion=Free
  twistLimitDegrees=45
  twistMotion=Free
 }
 {
  bone1Name=dyn_scarf_base
  bone2Name=pelvis
  swing1LimitDegrees=45
  swing1Motion=Free
  swing2LimitDegrees=45
  swing2Motion=Free
  twistLimitDegrees=45
  twistMotion=Free
 }
 {
  bone1Name=upperarm_l
  bone2Name=spine_05
  swing1LimitDegrees=45
  swing1Motion=Free
  swing2LimitDegrees=45
  swing2Motion=Free
  twistLimitDegrees=45
  twistMotion=Free
 }
 {
  bone1Name=upperarm_r
  bone2Name=clavicle_r
  swing1LimitDegrees=45
  swing1Motion=Free
  swing2LimitDegrees=45
  swing2Motion=Free
  twistLimitDegrees=45
  twistMotion=Free
 }
 {
  bone1Name=clavicle_r
  bone2Name=spine_05
  swing1LimitDegrees=45
  swing1Motion=Free
  swing2LimitDegrees=45
  swing2Motion=Free
  twistLimitDegrees=45
  twistMotion=Free
 }
]
