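# Post-process-style AnimBlueprint for the DashSurge medium male cosmetic body, adding secondary motion to scarf, shirt, sleeves and leek.
# Copied parent pose runs through Control Rig and AnimDynamics, then switchable RigidBody shirt setups and bone offsets.
# Cached poses feed a scarf stage blending simulation with constraint-based poses, then a DefaultSlot layer on dyn_leek.
# The EventGraph derives movement, backbling and front-end states, animates tassel noise, and drives material WPO cloth parameters.
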AnimBlueprint /BRCosmetics/Characters/Player/Male/Medium/Bodies/M_MED_DashSurge/Meshes/M_MED_DashSurge_AnimBP
target_skeleton=&/M_MED_DashSurge_Skeleton
graphs [
 AnimGraph {nodes [
   n1 = "Constraint - dyn_scarf_01 (PPPPPPPP)"!(
    ComponentPose=Local To Component(LocalPose=Use cached pose 'pre_scarf'())
    Alpha=1
    ConstraintWeights_0=0.4118
    ConstraintWeights_1:float=0.3333
    ConstraintWeights_2:float=0.0823
    ConstraintWeights_3:float=0.0757
    ConstraintWeights_4:float=0.0619
    ConstraintWeights_5:float=0.0182
    ConstraintWeights_6:float=0.0142
    ConstraintWeights_7:float=0.0027
   ) -> (Pose) {
    alphaBoolBlend {blendOption=Linear}
    boneToModify {boneName=dyn_scarf_01}
    constraintSetup [
     {targetBone {boneName=clavicle_l}}
     {targetBone {boneName=spine_05}}
     {targetBone {boneName=spine_03}}
     {targetBone {boneName=deform_upperarm_up_l}}
     {targetBone {boneName=upperarm_twist_01_l}}
     {targetBone {boneName=spine_04}}
     {targetBone {boneName=deform_upperarm_bk_l}}
     {targetBone {boneName=upperarm_twist_02_l}}
    ]
    constraintWeights [1,1,1,1,1,1,1,1]
   }
   n2 = "Transform (Modify) Bone - Bone: dyn_scarf_01"(ComponentPose=@n1, Alpha=1, Translation="0.118,-0.001,1.686", Rotation="-0.000,0.003,-0.034", Scale="1,1,1") {
    rotationMode=BMM_Additive
    rotationSpace=BCS_BoneSpace
    translationMode=BMM_Additive
    translationSpace=BCS_BoneSpace
    alphaBoolBlend {blendOption=Linear}
    boneToModify {boneName=dyn_scarf_01}
    rotation=<rot 0 0.003 -0.034>
    translation=<0.118 -0.001 1.686>
   }
   n3 = "Constraint - dyn_scarf_02 (PPPPPPPP)"!(
    ComponentPose=@n2
    Alpha=1
    ConstraintWeights_0=0.4296
    ConstraintWeights_1:float=0.2476
    ConstraintWeights_2:float=0.1049
    ConstraintWeights_3:float=0.0889
    ConstraintWeights_4:float=0.0583
    ConstraintWeights_5:float=0.0434
    ConstraintWeights_6:float=0.0194
    ConstraintWeights_7:float=0.0079
   ) -> (Pose) {
    alphaBoolBlend {blendOption=Linear}
    boneToModify {boneName=dyn_scarf_02}
    constraintSetup [
     {targetBone {boneName=spine_03}}
     {targetBone {boneName=clavicle_l}}
     {targetBone {boneName=spine_04}}
     {targetBone {boneName=deform_upperarm_bk_l}}
     {targetBone {boneName=upperarm_twist_01_l}}
     {targetBone {boneName=spine_05}}
     {targetBone {boneName=spine_02}}
     {targetBone {boneName=upperarm_twist_02_l}}
    ]
    constraintWeights [1,1,1,1,1,1,1,1]
   }
   n4 = "Constraint - dyn_scarf_03 (PPPPPPPP)"!(
    ComponentPose=@n3
    Alpha=1
    ConstraintWeights_0=0.3986
    ConstraintWeights_1:float=0.1882
    ConstraintWeights_2:float=0.1497
    ConstraintWeights_3:float=0.1461
    ConstraintWeights_4:float=0.057
    ConstraintWeights_5:float=0.0285
    ConstraintWeights_6:float=0.0179
    ConstraintWeights_7:float=0.0139
   ) -> (Pose) {
    alphaBoolBlend {blendOption=Linear}
    boneToModify {boneName=dyn_scarf_03}
    constraintSetup [
     {targetBone {boneName=spine_03}}
     {targetBone {boneName=spine_04}}
     {targetBone {boneName=clavicle_l}}
     {targetBone {boneName=spine_02}}
     {targetBone {boneName=spine_01}}
     {targetBone {boneName=deform_upperarm_bk_l}}
     {targetBone {boneName=pelvis}}
     {targetBone {boneName=spine_05}}
    ]
    constraintWeights [1,1,1,1,1,1,1,1]
   }
   n5 = "Constraint - dyn_scarf_04 (PPPPPPP)"!(
    ComponentPose=@n4
    Alpha=1
    ConstraintWeights_0=0.9674
    ConstraintWeights_1:float=0.0101
    ConstraintWeights_2:float=0.0077
    ConstraintWeights_3:float=0.0069
    ConstraintWeights_4:float=0.0064
    ConstraintWeights_5:float=0.0013
    ConstraintWeights_6:float=0.0001
   ) -> (Pose) {
    alphaBoolBlend {blendOption=Linear}
    boneToModify {boneName=dyn_scarf_04}
    constraintSetup [
     {targetBone {boneName=dyn_shirt_down_bk_l}}
     {targetBone {boneName=spine_02}}
     {targetBone {boneName=spine_03}}
     {targetBone {boneName=pelvis}}
     {targetBone {boneName=spine_01}}
     {targetBone {boneName=spine_04}}
     {targetBone {boneName=clavicle_l}}
    ]
    constraintWeights [1,1,1,1,1,1,1]
   }
   Blend Poses by bool(BlendPose_1=Component To Local(ComponentPose=@n5)) {blendType=Linear}
   n6 = RigidBody!(
    ComponentPose=Local To Component(LocalPose=Use cached pose 'pre_scarf'())
    Alpha
    OverrideWorldGravity:Vector="0,0,0"
    ExternalForce
    bTransferBoneVelocities:bool=False
   ) -> (Pose) {
    bClampLinearTranslationLimitToRefPose=true
    bOverrideWorldGravity=true
    bTransferBoneVelocities=true
    overridePhysicsAsset=&/M_MED_DashSurge_Scarf_Physics
    simulationSpace=BaseBoneSpace
    alphaBoolBlend {blendOption=Linear}
    baseBoneRef {boneName=spine_05}
    componentAppliedLinearAccClamp=<3000 3000 3000>
    componentLinearAccScale=<0.25 0.05 0.09>
    componentLinearVelScale=<0.25 0.05 0.09>
    overrideWorldGravity {z=-980}
    simSpaceSettings {
     maxAngularAcceleration=50
     maxAngularVelocity=1
     maxLinearAcceleration=300
     maxLinearVelocity=150
     velocityScaleZ=0.25
     worldAlpha=0.5
     externalLinearDragV=<0.2 0.2>
    }
   }
   n7 = "Constraint - dyn_scarf_01 (PPPPPPPP)"(
    ComponentPose=Local To Component(LocalPose=Use cached pose 'pre_scarf'())
    Alpha=1
    ConstraintWeights_0=0.4118
    ConstraintWeights_1=0.3333
    ConstraintWeights_2=0.0823
    ConstraintWeights_3=0.0757
    ConstraintWeights_4=0.0619
    ConstraintWeights_5=0.0182
    ConstraintWeights_6=0.0142
    ConstraintWeights_7=0.0027
   ) {
    alphaBoolBlend {blendOption=Linear}
    boneToModify {boneName=dyn_scarf_01}
    constraintSetup [
     {targetBone {boneName=clavicle_l}}
     {targetBone {boneName=spine_05}}
     {targetBone {boneName=spine_03}}
     {targetBone {boneName=deform_upperarm_up_l}}
     {targetBone {boneName=upperarm_twist_01_l}}
     {targetBone {boneName=spine_04}}
     {targetBone {boneName=deform_upperarm_bk_l}}
     {targetBone {boneName=upperarm_twist_02_l}}
    ]
    constraintWeights [1,1,1,1,1,1,1,1]
   }
   n8 = "Constraint - dyn_scarf_02 (PPPPPPPP)"(
    ComponentPose=@n7
    Alpha=1
    ConstraintWeights_0=0.4296
    ConstraintWeights_1=0.2476
    ConstraintWeights_2=0.1049
    ConstraintWeights_3=0.0889
    ConstraintWeights_4=0.0583
    ConstraintWeights_5=0.0434
    ConstraintWeights_6=0.0194
    ConstraintWeights_7=0.0079
   ) {
    alphaBoolBlend {blendOption=Linear}
    boneToModify {boneName=dyn_scarf_02}
    constraintSetup [
     {targetBone {boneName=spine_03}}
     {targetBone {boneName=clavicle_l}}
     {targetBone {boneName=spine_04}}
     {targetBone {boneName=deform_upperarm_bk_l}}
     {targetBone {boneName=upperarm_twist_01_l}}
     {targetBone {boneName=spine_05}}
     {targetBone {boneName=spine_02}}
     {targetBone {boneName=upperarm_twist_02_l}}
    ]
    constraintWeights [1,1,1,1,1,1,1,1]
   }
   n9 = "Constraint - dyn_scarf_03 (PPPPPPPP)"(
    ComponentPose=@n8
    Alpha=1
    ConstraintWeights_0=0.3986
    ConstraintWeights_1=0.1882
    ConstraintWeights_2=0.1497
    ConstraintWeights_3=0.1461
    ConstraintWeights_4=0.057
    ConstraintWeights_5=0.0285
    ConstraintWeights_6=0.0179
    ConstraintWeights_7=0.0139
   ) {
    alphaBoolBlend {blendOption=Linear}
    boneToModify {boneName=dyn_scarf_03}
    constraintSetup [
     {targetBone {boneName=spine_03}}
     {targetBone {boneName=spine_04}}
     {targetBone {boneName=clavicle_l}}
     {targetBone {boneName=spine_02}}
     {targetBone {boneName=spine_01}}
     {targetBone {boneName=deform_upperarm_bk_l}}
     {targetBone {boneName=pelvis}}
     {targetBone {boneName=spine_05}}
    ]
    constraintWeights [1,1,1,1,1,1,1,1]
   }
   n10 = "Constraint - dyn_scarf_04 (PPPPPPP)"(
    ComponentPose=@n9
    Alpha=1
    ConstraintWeights_0=0.9674
    ConstraintWeights_1=0.0101
    ConstraintWeights_2=0.0077
    ConstraintWeights_3=0.0069
    ConstraintWeights_4=0.0064
    ConstraintWeights_5=0.0013
    ConstraintWeights_6=0.0001
   ) {
    alphaBoolBlend {blendOption=Linear}
    boneToModify {boneName=dyn_scarf_04}
    constraintSetup [
     {targetBone {boneName=dyn_shirt_down_bk_l}}
     {targetBone {boneName=spine_02}}
     {targetBone {boneName=spine_03}}
     {targetBone {boneName=pelvis}}
     {targetBone {boneName=spine_01}}
     {targetBone {boneName=spine_04}}
     {targetBone {boneName=clavicle_l}}
    ]
    constraintWeights [1,1,1,1,1,1,1]
   }
   n11 = "Transform (Modify) Bone - Bone: dyn_scarf_04"(ComponentPose=@n10, Alpha=1, Translation="0,0,-2", Rotation="0,0,0", Scale="1,1,1") {
    rotationMode=BMM_Additive
    rotationSpace=BCS_BoneSpace
    translationMode=BMM_Additive
    translationSpace=BCS_BoneSpace
    alphaBoolBlend {blendOption=Linear}
    boneToModify {boneName=dyn_scarf_04}
    translation {z=-2}
   }
   n12 = Blend Poses by bool(BlendPose_0=Component To Local(ComponentPose=@n11), BlendPose_1=Component To Local(ComponentPose=@n6)) {blendType=Linear}
   n13 = "Constraint - dyn_scarf_01 (PPPPPPPP)"(
    ComponentPose=Local To Component(LocalPose=Use cached pose 'pre_scarf'())
    Alpha=1
    ConstraintWeights_0=0.4118
    ConstraintWeights_1=0.3333
    ConstraintWeights_2=0.0823
    ConstraintWeights_3=0.0757
    ConstraintWeights_4=0.0619
    ConstraintWeights_5=0.0182
    ConstraintWeights_6=0.0142
    ConstraintWeights_7=0.0027
   ) {
    alphaBoolBlend {blendOption=Linear}
    boneToModify {boneName=dyn_scarf_01}
    constraintSetup [
     {targetBone {boneName=clavicle_l}}
     {targetBone {boneName=spine_05}}
     {targetBone {boneName=spine_03}}
     {targetBone {boneName=deform_upperarm_up_l}}
     {targetBone {boneName=upperarm_twist_01_l}}
     {targetBone {boneName=spine_04}}
     {targetBone {boneName=deform_upperarm_bk_l}}
     {targetBone {boneName=upperarm_twist_02_l}}
    ]
    constraintWeights [1,1,1,1,1,1,1,1]
   }
   n14 = "Constraint - dyn_scarf_02 (PPPPPPPP)"(
    ComponentPose=@n13
    Alpha=1
    ConstraintWeights_0=0.4296
    ConstraintWeights_1=0.2476
    ConstraintWeights_2=0.1049
    ConstraintWeights_3=0.0889
    ConstraintWeights_4=0.0583
    ConstraintWeights_5=0.0434
    ConstraintWeights_6=0.0194
    ConstraintWeights_7=0.0079
   ) {
    alphaBoolBlend {blendOption=Linear}
    boneToModify {boneName=dyn_scarf_02}
    constraintSetup [
     {targetBone {boneName=spine_03}}
     {targetBone {boneName=clavicle_l}}
     {targetBone {boneName=spine_04}}
     {targetBone {boneName=deform_upperarm_bk_l}}
     {targetBone {boneName=upperarm_twist_01_l}}
     {targetBone {boneName=spine_05}}
     {targetBone {boneName=spine_02}}
     {targetBone {boneName=upperarm_twist_02_l}}
    ]
    constraintWeights [1,1,1,1,1,1,1,1]
   }
   n15 = "Constraint - dyn_scarf_03 (PPPPPPPP)"(
    ComponentPose=@n14
    Alpha=1
    ConstraintWeights_0=0.3986
    ConstraintWeights_1=0.1882
    ConstraintWeights_2=0.1497
    ConstraintWeights_3=0.1461
    ConstraintWeights_4=0.057
    ConstraintWeights_5=0.0285
    ConstraintWeights_6=0.0179
    ConstraintWeights_7=0.0139
   ) {
    alphaBoolBlend {blendOption=Linear}
    boneToModify {boneName=dyn_scarf_03}
    constraintSetup [
     {targetBone {boneName=spine_03}}
     {targetBone {boneName=spine_04}}
     {targetBone {boneName=clavicle_l}}
     {targetBone {boneName=spine_02}}
     {targetBone {boneName=spine_01}}
     {targetBone {boneName=deform_upperarm_bk_l}}
     {targetBone {boneName=pelvis}}
     {targetBone {boneName=spine_05}}
    ]
    constraintWeights [1,1,1,1,1,1,1,1]
   }
   n16 = "Transform (Modify) Bone - Bone: dyn_scarf_04"(ComponentPose=@n15, Alpha=1, Translation="0,0,-1", Rotation="-10,0,0", Scale="1,1,1") {
    rotationMode=BMM_Additive
    rotationSpace=BCS_BoneSpace
    translationMode=BMM_Additive
    translationSpace=BCS_BoneSpace
    alphaBoolBlend {blendOption=Linear}
    boneToModify {boneName=dyn_scarf_04}
    rotation {pitch=-10}
    translation {z=-1}
   }
   n17 = "Transform (Modify) Bone - Bone: dyn_scarf_02"(ComponentPose=@n16, Alpha=1, Translation="0,0,4", Rotation="0,0,0", Scale="1,1,1") {
    rotationMode=BMM_Additive
    rotationSpace=BCS_BoneSpace
    translationMode=BMM_Additive
    translationSpace=BCS_BoneSpace
    alphaBoolBlend {blendOption=Linear}
    boneToModify {boneName=dyn_scarf_02}
    translation {z=4}
   }
   n18 = "Transform (Modify) Bone - Bone: dyn_scarf_04"(ComponentPose=@n17, Alpha=1, Translation="0,0,2", Rotation="20,0,0", Scale="1,1,1") {
    rotationMode=BMM_Additive
    rotationSpace=BCS_BoneSpace
    translationMode=BMM_Additive
    translationSpace=BCS_BoneSpace
    alphaBoolBlend {blendOption=Linear}
    boneToModify {boneName=dyn_scarf_04}
    rotation {pitch=20}
    translation {z=2}
   }
   n19 = Component To Local(ComponentPose=@n18)
   n20 = "Transform (Modify) Bone - Bone: dyn_sleeve_bottom_r"!(
    ComponentPose=Local To Component(LocalPose=Blend Poses by bool(BlendPose_0=@n19, BlendPose_1=@n12) {blendType=Linear})
    bAlphaBoolEnabled:bool=True
    Translation="0,0,-15"
    Rotation="30,0,0"
    Scale="1,1,1"
   ) -> (Pose) {
    alphaInputType=Bool
    rotationMode=BMM_Additive
    rotationSpace=BCS_BoneSpace
    translationMode=BMM_Additive
    translationSpace=BCS_BoneSpace
    alphaBoolBlend {blendOption=Linear}
    boneToModify {boneName=dyn_sleeve_bottom_r}
    rotation {pitch=30}
    translation {z=-15}
   }
   Output Pose(Result=Layered blend per bone(
     BasePose=Component To Local(ComponentPose=@n20)
     BlendPoses_0=Slot 'DefaultSlot'()
     BlendWeights_0=Get FECustomIdleBlend() -> (FECustomIdleBlend:float)
    ) {layerSetup [{branchFilters [{boneName=dyn_leek}]}]}) {layerGroup=None}
   n21 = Control Rig(Source=Copy Pose From Mesh() {bUseAttachedParent=true}) {
    alphaBoolBlend {blendOption=Linear}
    controlRigAssetReference {blueprintRigClass=&/M_MED_DashSurge_CtrlRig.M_MED_DashSurge_CtrlRig_C}
   }
   n22 = Anim Dynamics - Bone: dyn_shirt_down_bk_r!(ComponentPose=Local To Component(LocalPose=@n21), Alpha=0.2, GravityOverride:Vector="0,0,0", ExternalForce="0,0,0") -> (Pose) {
    angularDampingOverride=1
    angularSpringConstant=20000
    bAngularSpring=true
    bOverrideAngularDamping=true
    bUseGravityOverride=true
    gravityScale=0.35
    simulationSpace=RootRelative
    alphaBoolBlend {blendOption=Linear}
    boundBone {boneName=dyn_shirt_down_bk_r}
    componentAppliedLinearAccClamp=<6000 6000 6000>
    componentLinearAccScale=<1 1 1>
    componentLinearVelScale=<1 1 1>
    gravityOverride {z=-980}
    physicsBodyDefinitions [{
      boundBone {boneName=dyn_shirt_down_bk_r}
      boxExtents=<20 20 20>
      constraintSetup {angularLimitsMax=<15 15 15>, angularLimitsMin=<-15 -15 -15>, angularTarget=<1 0 0>}
      localJointOffset=<4 0 0>
     }]
    relativeSpaceBone {boneName=pelvis}
   }
   n23 = Anim Dynamics - Bone: dyn_shirt_down_fr_r!(ComponentPose=@n22, Alpha=0.2, GravityOverride:Vector="0,0,-980", ExternalForce="0,0,0") -> (Pose) {
    angularDampingOverride=1
    angularSpringConstant=20000
    bAngularSpring=true
    bOverrideAngularDamping=true
    bUseGravityOverride=true
    gravityScale=0.35
    simulationSpace=RootRelative
    alphaBoolBlend {blendOption=Linear}
    boundBone {boneName=dyn_shirt_down_fr_r}
    componentAppliedLinearAccClamp=<6000 6000 6000>
    componentLinearAccScale=<1 1 1>
    componentLinearVelScale=<1 1 1>
    gravityOverride {z=-980}
    physicsBodyDefinitions [{
      boundBone {boneName=dyn_shirt_down_fr_r}
      boxExtents=<20 20 20>
      constraintSetup {angularLimitsMax=<15 15 15>, angularLimitsMin=<-15 -15 -15>, angularTarget=<1 0 0>}
      localJointOffset=<4 0 0>
     }]
    relativeSpaceBone {boneName=pelvis}
   }
   n24 = Anim Dynamics - Bone: dyn_shirt_down_side_r!(ComponentPose=@n23, Alpha=0.2, GravityOverride:Vector="0,0,-980", ExternalForce="0,0,0") -> (Pose) {
    angularDampingOverride=1
    angularSpringConstant=20000
    bAngularSpring=true
    bOverrideAngularDamping=true
    bUseGravityOverride=true
    gravityScale=0.35
    simulationSpace=RootRelative
    alphaBoolBlend {blendOption=Linear}
    boundBone {boneName=dyn_shirt_down_side_r}
    componentAppliedLinearAccClamp=<6000 6000 6000>
    componentLinearAccScale=<1 1 1>
    componentLinearVelScale=<1 1 1>
    gravityOverride {z=-980}
    physicsBodyDefinitions [{
      boundBone {boneName=dyn_shirt_down_side_r}
      boxExtents=<20 20 20>
      constraintSetup {angularLimitsMax=<15 15 15>, angularLimitsMin=<-15 -15 -15>, angularTarget=<1 0 0>}
      localJointOffset=<4 0 0>
     }]
    relativeSpaceBone {boneName=pelvis}
   }
   n25 = Anim Dynamics - Bone: dyn_shirt_down_bk_l!(ComponentPose=@n24, Alpha=0.2, GravityOverride:Vector="0,0,-980", ExternalForce="0,0,0") -> (Pose) {
    angularDampingOverride=1
    angularSpringConstant=20000
    bAngularSpring=true
    bOverrideAngularDamping=true
    bUseGravityOverride=true
    gravityScale=0.35
    simulationSpace=RootRelative
    alphaBoolBlend {blendOption=Linear}
    boundBone {boneName=dyn_shirt_down_bk_l}
    componentAppliedLinearAccClamp=<6000 6000 6000>
    componentLinearAccScale=<1 1 1>
    componentLinearVelScale=<1 1 1>
    gravityOverride {z=-980}
    physicsBodyDefinitions [{
      boundBone {boneName=dyn_shirt_down_bk_l}
      boxExtents=<10 10 10>
      constraintSetup {angularLimitsMax=<15 15 15>, angularLimitsMin=<-15 -15 -15>, angularTarget=<1 0 0>}
      localJointOffset=<4 0 0>
     }]
    relativeSpaceBone {boneName=pelvis}
   }
   n26 = Anim Dynamics - Bone: dyn_shirt_down_fr_l!(ComponentPose=@n25, Alpha=0.2, GravityOverride:Vector="0,0,-980", ExternalForce="0,0,0") -> (Pose) {
    angularDampingOverride=1
    angularSpringConstant=20000
    bAngularSpring=true
    bOverrideAngularDamping=true
    bUseGravityOverride=true
    gravityScale=0.35
    simulationSpace=RootRelative
    alphaBoolBlend {blendOption=Linear}
    boundBone {boneName=dyn_shirt_down_fr_l}
    componentAppliedLinearAccClamp=<6000 6000 6000>
    componentLinearAccScale=<1 1 1>
    componentLinearVelScale=<1 1 1>
    gravityOverride {z=-980}
    physicsBodyDefinitions [{
      boundBone {boneName=dyn_shirt_down_fr_l}
      boxExtents=<20 20 20>
      constraintSetup {angularLimitsMax=<15 15 15>, angularLimitsMin=<-15 -15 -15>, angularTarget=<1 0 0>}
      localJointOffset=<4 0 0>
     }]
    relativeSpaceBone {boneName=pelvis}
   }
   n27 = Anim Dynamics - Bone: dyn_shirt_down_side_l!(ComponentPose=@n26, Alpha=0.2, GravityOverride:Vector="0,0,-980", ExternalForce="0,0,0") -> (Pose) {
    angularDampingOverride=1
    angularSpringConstant=20000
    bAngularSpring=true
    bOverrideAngularDamping=true
    bUseGravityOverride=true
    gravityScale=0.35
    simulationSpace=RootRelative
    alphaBoolBlend {blendOption=Linear}
    boundBone {boneName=dyn_shirt_down_side_l}
    componentAppliedLinearAccClamp=<6000 6000 6000>
    componentLinearAccScale=<1 1 1>
    componentLinearVelScale=<1 1 1>
    gravityOverride {z=-980}
    physicsBodyDefinitions [{
      boundBone {boneName=dyn_shirt_down_side_l}
      boxExtents=<30 30 30>
      constraintSetup {angularLimitsMax=<15 15 15>, angularLimitsMin=<-15 -15 -15>, angularTarget=<1 0 0>}
      localJointOffset=<4 0 0>
     }]
    relativeSpaceBone {boneName=pelvis}
   }
   n28 = Anim Dynamics - Bone: dyn_sleeve_top_l!(ComponentPose=@n27, Alpha=1, GravityOverride:Vector="0,0,-980", ExternalForce="0,0,0") -> (Pose) {
    angularDampingOverride=1
    angularSpringConstant=20000
    bAngularSpring=true
    bGravityOverrideInSimSpace=true
    bOverrideAngularDamping=true
    bUseGravityOverride=true
    gravityScale=0.35
    simulationSpace=RootRelative
    alphaBoolBlend {blendOption=Linear}
    boundBone {boneName=dyn_sleeve_top_l}
    componentAppliedLinearAccClamp=<6000 6000 600>
    componentLinearAccScale=<1 1 1>
    componentLinearVelScale=<1 1 1>
    gravityOverride {z=-980}
    physicsBodyDefinitions [{
      boundBone {boneName=dyn_sleeve_top_l}
      boxExtents=<20 30 30>
      constraintSetup {angularLimitsMax=<15 15 15>, angularLimitsMin=<-15 -15 -15>, angularTarget=<1 0 0>}
      localJointOffset=<4 0 0>
     }]
   }
   n29 = Anim Dynamics - Bone: dyn_sleeve_back_l!(ComponentPose=@n28, Alpha=1, GravityOverride:Vector="0,0,-980", ExternalForce="0,0,0") -> (Pose) {
    angularDampingOverride=1
    angularSpringConstant=20000
    bAngularSpring=true
    bGravityOverrideInSimSpace=true
    bOverrideAngularDamping=true
    bUseGravityOverride=true
    gravityScale=0.35
    simulationSpace=RootRelative
    alphaBoolBlend {blendOption=Linear}
    boundBone {boneName=dyn_sleeve_back_l}
    componentAppliedLinearAccClamp=<6000 6000 600>
    componentLinearAccScale=<1 1 1>
    componentLinearVelScale=<1 1 1>
    gravityOverride {z=-980}
    physicsBodyDefinitions [{
      boundBone {boneName=dyn_sleeve_back_l}
      boxExtents=<20 30 30>
      constraintSetup {angularLimitsMax=<15 15 15>, angularLimitsMin=<-15 -15 -15>, angularTarget=<1 0 0>}
      localJointOffset=<4 0 0>
     }]
   }
   n30 = Anim Dynamics - Bone: dyn_sleeve_front_l!(ComponentPose=@n29, Alpha=1, GravityOverride:Vector="0,0,-980", ExternalForce="0,0,0") -> (Pose) {
    angularDampingOverride=1
    angularSpringConstant=20000
    bAngularSpring=true
    bGravityOverrideInSimSpace=true
    bOverrideAngularDamping=true
    bUseGravityOverride=true
    gravityScale=0.35
    simulationSpace=RootRelative
    alphaBoolBlend {blendOption=Linear}
    boundBone {boneName=dyn_sleeve_front_l}
    componentAppliedLinearAccClamp=<6000 6000 600>
    componentLinearAccScale=<1 1 1>
    componentLinearVelScale=<1 1 1>
    gravityOverride {z=-980}
    physicsBodyDefinitions [{
      boundBone {boneName=dyn_sleeve_front_l}
      boxExtents=<20 30 30>
      constraintSetup {angularLimitsMax=<15 15 15>, angularLimitsMin=<-15 -15 -15>, angularTarget=<1 0 0>}
      localJointOffset=<4 0 0>
     }]
   }
   n31 = Anim Dynamics - Bone: dyn_sleeve_bottom_l!(ComponentPose=@n30, Alpha=1, GravityOverride:Vector="0,0,-980", ExternalForce="0,0,0") -> (Pose) {
    angularDampingOverride=1
    angularSpringConstant=20000
    bAngularSpring=true
    bGravityOverrideInSimSpace=true
    bOverrideAngularDamping=true
    bUseGravityOverride=true
    gravityScale=0.35
    simulationSpace=RootRelative
    alphaBoolBlend {blendOption=Linear}
    boundBone {boneName=dyn_sleeve_bottom_l}
    componentAppliedLinearAccClamp=<6000 6000 600>
    componentLinearAccScale=<1 1 1>
    componentLinearVelScale=<1 1 1>
    gravityOverride {z=-980}
    physicsBodyDefinitions [{
      boundBone {boneName=dyn_sleeve_bottom_l}
      boxExtents=<20 30 30>
      constraintSetup {angularLimitsMax=<15 15 15>, angularLimitsMin=<-15 -15 -15>, angularTarget=<1 0 0>}
      localJointOffset=<4 0 0>
     }]
   }
   n32 = Anim Dynamics - Bone: dyn_sleeve_bottom_r!(ComponentPose=@n31, Alpha=1, GravityOverride:Vector="0,0,-980", ExternalForce="0,0,0") -> (Pose) {
    angularDampingOverride=1
    angularSpringConstant=20000
    bAngularSpring=true
    bGravityOverrideInSimSpace=true
    bOverrideAngularDamping=true
    bUseGravityOverride=true
    gravityScale=0.35
    simulationSpace=RootRelative
    alphaBoolBlend {blendOption=Linear}
    boundBone {boneName=dyn_sleeve_bottom_r}
    componentAppliedLinearAccClamp=<6000 6000 600>
    componentLinearAccScale=<1 1 1>
    componentLinearVelScale=<1 1 1>
    gravityOverride {z=-980}
    physicsBodyDefinitions [{
      boundBone {boneName=dyn_sleeve_bottom_r}
      boxExtents=<20 30 30>
      constraintSetup {angularLimitsMax=<15 15 15>, angularLimitsMin=<-15 -15 -15>, angularTarget=<1 0 0>}
      localJointOffset=<4 0 0>
     }]
   }
   n33 = Anim Dynamics - Bone: dyn_sleeve_front_r!(ComponentPose=@n32, Alpha=1, GravityOverride:Vector="0,0,-980", ExternalForce="0,0,0") -> (Pose) {
    angularDampingOverride=1
    angularSpringConstant=20000
    bAngularSpring=true
    bGravityOverrideInSimSpace=true
    bOverrideAngularDamping=true
    bUseGravityOverride=true
    gravityScale=0.35
    simulationSpace=RootRelative
    alphaBoolBlend {blendOption=Linear}
    boundBone {boneName=dyn_sleeve_front_r}
    componentAppliedLinearAccClamp=<6000 6000 600>
    componentLinearAccScale=<1 1 1>
    componentLinearVelScale=<1 1 1>
    gravityOverride {z=-980}
    physicsBodyDefinitions [{
      boundBone {boneName=dyn_sleeve_front_r}
      boxExtents=<20 30 30>
      constraintSetup {angularLimitsMax=<15 15 15>, angularLimitsMin=<-15 -15 -15>, angularTarget=<1 0 0>}
      localJointOffset=<4 0 0>
     }]
    sphericalLimits [{limitRadius=5, drivingBone {boneName=lowerarm_r}, sphereLocalOffset=<-23.3832 -10.1102 -3.6112>}]
   }
   n34 = Anim Dynamics - Bone: dyn_sleeve_back_r!(ComponentPose=@n33, Alpha=1, GravityOverride:Vector="0,0,-980", ExternalForce="0,0,0") -> (Pose) {
    angularDampingOverride=1
    angularSpringConstant=20000
    bAngularSpring=true
    bGravityOverrideInSimSpace=true
    bOverrideAngularDamping=true
    bUseGravityOverride=true
    gravityScale=0.35
    simulationSpace=RootRelative
    alphaBoolBlend {blendOption=Linear}
    boundBone {boneName=dyn_sleeve_back_r}
    componentAppliedLinearAccClamp=<6000 6000 600>
    componentLinearAccScale=<1 1 1>
    componentLinearVelScale=<1 1 1>
    gravityOverride {z=-980}
    physicsBodyDefinitions [{
      boundBone {boneName=dyn_sleeve_back_r}
      boxExtents=<20 30 30>
      constraintSetup {angularLimitsMax=<15 15 15>, angularLimitsMin=<-15 -15 -15>, angularTarget=<1 0 0>}
      localJointOffset=<4 0 0>
     }]
   }
   n35 = Anim Dynamics - Bone: dyn_sleeve_top_r!(ComponentPose=@n34, Alpha=1, GravityOverride:Vector="0,0,-980", ExternalForce="0,0,0") -> (Pose) {
    angularDampingOverride=1
    angularSpringConstant=20000
    bAngularSpring=true
    bGravityOverrideInSimSpace=true
    bOverrideAngularDamping=true
    bUseGravityOverride=true
    gravityScale=0.35
    simulationSpace=RootRelative
    alphaBoolBlend {blendOption=Linear}
    boundBone {boneName=dyn_sleeve_top_r}
    componentAppliedLinearAccClamp=<6000 6000 600>
    componentLinearAccScale=<1 1 1>
    componentLinearVelScale=<1 1 1>
    gravityOverride {z=-980}
    physicsBodyDefinitions [{
      boundBone {boneName=dyn_sleeve_top_r}
      boxExtents=<20 30 30>
      constraintSetup {angularLimitsMax=<15 15 15>, angularLimitsMin=<-15 -15 -15>, angularTarget=<1 0 0>}
      localJointOffset=<4 0 0>
     }]
   }
   n36 = "Transform (Modify) Bone - Bone: dyn_leek_base"!(ComponentPose=@n35, bAlphaBoolEnabled:bool=True, Translation="0,0,0", Rotation="10,10,0", Scale="1,1,1") -> (Pose) {
    alphaInputType=Bool
    rotationMode=BMM_Additive
    rotationSpace=BCS_BoneSpace
    translationMode=BMM_Additive
    translationSpace=BCS_BoneSpace
    alphaBoolBlend {blendInTime=0.8, blendOption=Linear, blendOutTime=0.8}
    alphaScaleBiasClamp {bInterpResult=true, interpSpeedDecreasing=6, interpSpeedIncreasing=0}
    boneToModify {boneName=dyn_leek_base}
    rotation {pitch=10, yaw=10}
   }
   Save cached pose 'pre_dyn_shirt'(Pose=Component To Local(ComponentPose=@n36)) {cacheName=pre_dyn_shirt}
   n37 = "Transform (Modify) Bone - Bone: dyn_scarf_base"(
    ComponentPose=Local To Component(LocalPose=Use cached pose 'post_dyn_shirt'())
    Alpha=Get Tassle Alpha() -> (Tassle Alpha:double)
    Translation="0,0,0"
    Rotation=Get Tassle Rotator() -> (Tassle Rotator:Rotator)
    Scale="1,1,1"
   ) {
    rotationMode=BMM_Additive
    rotationSpace=BCS_BoneSpace
    translationSpace=BCS_BoneSpace
    alphaBoolBlend {blendOption=Linear}
    alphaScaleBiasClamp {interpSpeedDecreasing=2, interpSpeedIncreasing=2}
    boneToModify {boneName=dyn_scarf_base}
   }
   n38 = "Transform (Modify) Bone - Bone: dyn_scarf_base"!(ComponentPose=@n37, bAlphaBoolEnabled:bool=True, Translation="50,0,0", Rotation="0,0,-40", Scale="1,1,1") -> (Pose) {
    alphaInputType=Bool
    rotationMode=BMM_Additive
    rotationSpace=BCS_BoneSpace
    translationSpace=BCS_BoneSpace
    alphaBoolBlend {blendOption=Linear}
    alphaScaleBiasClamp {interpSpeedDecreasing=2, interpSpeedIncreasing=2}
    boneToModify {boneName=dyn_scarf_base}
    rotation {roll=-40}
    translation {x=50}
   }
   Save cached pose 'pre_scarf'(Pose=Component To Local(ComponentPose=@n38)) {cacheName=pre_scarf}
   n39 = RigidBody!(ComponentPose=Local To Component(LocalPose=Use cached pose 'pre_dyn_shirt'()), Alpha, OverrideWorldGravity="0,0,0", ExternalForce) -> (Pose) {
    bClampLinearTranslationLimitToRefPose=true
    bFreezeIncomingPoseOnStart=true
    bOverrideWorldGravity=true
    overridePhysicsAsset=&/M_MED_DashSurge_Crouch_Physics
    simulationSpace=BaseBoneSpace
    alphaBoolBlend {blendOption=Linear}
    baseBoneRef {boneName=pelvis}
    componentAppliedLinearAccClamp=<3000 3000 3000>
    componentLinearAccScale=<0.25 0.05 0.09>
    componentLinearVelScale=<0.25 0.05 0.09>
    overrideWorldGravity {z=-980}
    simSpaceSettings {
     maxAngularAcceleration=50
     maxAngularVelocity=1
     maxLinearAcceleration=300
     maxLinearVelocity=150
     velocityScaleZ=0.25
     worldAlpha=0.5
     externalLinearDragV=<0.2 0.2>
    }
   }
   n40 = Component To Local(ComponentPose=@n39)
   n41 = RigidBody(ComponentPose=Local To Component(LocalPose=Use cached pose 'pre_dyn_shirt'()), OverrideWorldGravity="0,0,0") {
    bClampLinearTranslationLimitToRefPose=true
    bFreezeIncomingPoseOnStart=true
    bOverrideWorldGravity=true
    overridePhysicsAsset=&/M_MED_DashSurge_Physics
    simulationSpace=BaseBoneSpace
    alphaBoolBlend {blendOption=Linear}
    baseBoneRef {boneName=pelvis}
    componentAppliedLinearAccClamp=<3000 3000 3000>
    componentLinearAccScale=<0.25 0.05 0.09>
    componentLinearVelScale=<0.25 0.05 0.09>
    overrideWorldGravity {z=-980}
    simSpaceSettings {
     maxAngularAcceleration=50
     maxAngularVelocity=1
     maxLinearAcceleration=300
     maxLinearVelocity=150
     velocityScaleZ=0.25
     worldAlpha=0.5
     externalAngularVelocity=<0.2 0.2>
     externalLinearDragV=<0.2 0.2>
     externalLinearVelocity=<0.2 0.2>
    }
   }
   n42 = Component To Local(ComponentPose=@n41)
   n43 = "Transform (Modify) Bone - Bone: extra_leek"!(
    ComponentPose=Local To Component(LocalPose=Blend Poses by bool(BlendPose_0=@n40, BlendPose_1=@n42) {blendType=Linear})
    bAlphaBoolEnabled:bool=True
    Translation="5.014,-2.026,-0.225"
    Rotation="-6.850,76.407,-0.975"
    Scale="1,1,1"
   ) -> (Pose) {
    alphaInputType=Bool
    rotationMode=BMM_Additive
    rotationSpace=BCS_BoneSpace
    translationMode=BMM_Additive
    translationSpace=BCS_BoneSpace
    alphaBoolBlend {blendOption=Linear}
    alphaScaleBiasClamp {interpSpeedDecreasing=2, interpSpeedIncreasing=2}
    boneToModify {boneName=extra_leek}
    rotation=<rot -6.85 76.4069 -0.975>
    translation=<5.014 -2.026 -0.225>
   }
   n44 = "Transform (Modify) Bone - Bone: dyn_shirt_bk_l_base"!(ComponentPose=@n43, bAlphaBoolEnabled:bool=True, Translation="0,0,0", Rotation="-35.000,0.000,-0.000", Scale="1,1,1") -> (Pose) {
    alphaInputType=Bool
    rotationMode=BMM_Additive
    rotationSpace=BCS_BoneSpace
    translationMode=BMM_Additive
    translationSpace=BCS_BoneSpace
    alphaBoolBlend {blendInTime=0.8, blendOption=Linear}
    alphaScaleBiasClamp {interpSpeedDecreasing=2, interpSpeedIncreasing=2}
    boneToModify {boneName=dyn_shirt_bk_l_base}
    rotation=<rot -35 0.0001 0>
   }
   n45 = "Transform (Modify) Bone - Bone: dyn_shirt_bk_r_base"!(ComponentPose=@n44, bAlphaBoolEnabled:bool=True, Translation="0,0,0", Rotation="-60.000,0.000,-0.000", Scale="1,1,1") -> (Pose) {
    alphaInputType=Bool
    rotationMode=BMM_Additive
    rotationSpace=BCS_BoneSpace
    translationMode=BMM_Additive
    translationSpace=BCS_BoneSpace
    alphaBoolBlend {blendInTime=0.8, blendOption=Linear}
    alphaScaleBiasClamp {interpSpeedDecreasing=2, interpSpeedIncreasing=2}
    boneToModify {boneName=dyn_shirt_bk_r_base}
    rotation=<rot -60 0 0>
   }
   n46 = "Transform (Modify) Bone - Bone: dyn_shirt_bk_cntr_base"!(ComponentPose=@n45, bAlphaBoolEnabled:bool=True, Translation="0,0,0", Rotation="-70.000,0.000,-0.000", Scale="1,1,1") -> (Pose) {
    alphaInputType=Bool
    rotationMode=BMM_Additive
    rotationSpace=BCS_BoneSpace
    translationMode=BMM_Additive
    translationSpace=BCS_BoneSpace
    alphaBoolBlend {blendInTime=0.8, blendOption=Linear}
    alphaScaleBiasClamp {interpSpeedDecreasing=2, interpSpeedIncreasing=2}
    boneToModify {boneName=dyn_shirt_bk_cntr_base}
    rotation {pitch=-70}
   }
   n47 = "Transform (Modify) Bone - Bone: dyn_shirt_side_r_base"!(ComponentPose=@n46, bAlphaBoolEnabled:bool=True, Translation="0,0,0", Rotation="59.437,24.162,64.405", Scale="1,1,1") -> (Pose) {
    alphaInputType=Bool
    rotationMode=BMM_Additive
    rotationSpace=BCS_BoneSpace
    translationMode=BMM_Additive
    translationSpace=BCS_BoneSpace
    alphaBoolBlend {blendInTime=0.8, blendOption=Linear}
    alphaScaleBiasClamp {interpSpeedDecreasing=2, interpSpeedIncreasing=2}
    boneToModify {boneName=dyn_shirt_side_r_base}
    rotation=<rot 59.437 24.162 64.405>
   }
   Save cached pose 'post_dyn_shirt'(Pose=Component To Local(ComponentPose=@n47)) {cacheName=post_dyn_shirt}
  ]}
 EventGraph {nodes [
   Event OnCharacterCustomizationLoaded()
   > n1 = Sequence()
   n2 = Set SkelMesh(execute=@n1.then_1, SkelMesh=GetPartSkeletalMeshComponent())
   > n3 = Get Dynamic Material Instance(SkeletalMeshComponent=@n2)
   > Set Mat Instance Dynamic(Mat Instance Dynamic=@n3)
   > n4 = GetOrCreateDynamicOverlayMaterial(MeshComponent=@n2, MaterialName=None)
   > Set Overlay Mat Instance Dynamic(Overlay Mat Instance Dynamic=@n4)
   Event BlueprintInitializeAnimation()
   > Set SkelMesh(SkelMesh=GetOwnerSkeletalMeshComponent())
   n5 = TryGetPawnOwner()
   > Is Valid(InputObject=@n5)
   > n6 = Cast To FortPlayerPawnAthena(Object=@n5)
   > Set PlayerPawn(PlayerPawn:Fort Player Pawn Athena Object Reference=@n6) -> (Output_Get:Fort Player Pawn Athena Object Reference)
   n7 = Event BlueprintUpdateAnimation()
   > n8 = Sequence!(execute) -> (then_0, then_1, then_2:Exec, then_3:Exec)
   n9 = Sequence!(execute=@n8.then_1) -> (then_0, then_1, then_2, then_3, then_4:Exec, then_5:Exec, then_6:Exec)
   n10 = Gravity Override State Function DEPRECATED(
    execute=@n9.then_1
    CharacterPartAnimInstance=Self-Reference()
    GameTimeInSeconds=GetGameTimeInSeconds()
    GravityOverrideParams="Get  Tassle Gravity Override Params"() -> (" Tassle Gravity Override Params":Gravity Override Params Struct Structure)
   )
   > n11 = Branch(Condition=Get Is JumpGround() -> (Is JumpGround:bool))
   > "Set  Tassle GravityOverride"(" Tassle GravityOverride":Vector="0,0,-400") -> (Output_Get:Vector)
   "Set  Tassle GravityOverride"(execute=@n11.else, " Tassle GravityOverride"=@n10)
   n12 = RandomRotatorFunction!(
    execute=@n9.then_2
    GameTimeInSeconds:double=float * float(A=GetGameTimeInSeconds(), B=1.125) {is=title K2Node_PromotableOperator}
    PerlinScale:double=1
    Frequency:Rotator="5,2.8,5"
    Amplitude:Rotator="10,5,3"
    Offset:Rotator="10,3,0"
    TimeOffset:double=0
   ) -> (then, RandomRotator:Rotator)
   > Set Tassle Rotator(Tassle Rotator:Rotator=@n12) -> (Output_Get:Rotator)
   n13 = Branch(execute=@n9.then_3, Condition=Get bIsInFrontEnd())
   > Set Tassle Alpha(Tassle Alpha:double=0.25) -> (Output_Get:double)
   Set Tassle Alpha(execute=@n13.else, Tassle Alpha=0.7)
   n14 = Sequence!(execute=@n9.then_4) -> (then_0, then_1)
   n15 = CheckForBackbling()
   > n16 = Branch(Condition=@n15)
   > Set Is Wearing Backbling(Is Wearing Backbling=true)
   Set Is Wearing Backbling(execute=@n16.else, Is Wearing Backbling=false)
   n17 = Branch(execute=@n14.then_1, Condition=@n15.bIsBackblingACape)
   > Set Is Backbling Cape(Is Backbling Cape:bool=true) -> (Output_Get:bool)
   Set Is Backbling Cape(execute=@n17.else, Is Backbling Cape=false)
   n18 = Get bIsSkyDiving()
   n19 = Get bIsFalling()
   n20 = Get bIsJumping()
   Set Is JumpGround(execute=@n9.then_0, Is JumpGround:bool=AND Boolean(A=OR Boolean(A=@n20, B=@n19), B=NOT Boolean(A=@n18))) -> (Output_Get:bool)
   > Set Rotate spine(Rotate spine:float=GetAlphaFromReferenceTransformDelta(
     SkeletalMeshComponent=GetOwnerSkeletalMeshComponent()
     BoneName=spine_05
     BaseName=pelvis
     DeltaComponentType=EulerX
     InputRange=(X=-20,Y=-40)
     TwistAxis="1,0,0"
    )) -> (Output_Get:float)
   n21 = Get bIsPlayingEmote()
   Branch(execute=@n9.then_6, Condition=AND Boolean(A=NOT Boolean(A=@n21), B=Get bIsInFrontEnd()))
   > Sequence()
   n22 = GetOwningComponent()
   > n23 = Branch(Condition="Equal (Float)"(A=GetAlphaFromReferenceTransformDelta(SkeletalMeshComponent=@n22, BoneName=upperarm_l, BaseName=lowerarm_l, InputRange=(X=-20,Y=-30), TwistAxis="1,0,0")))
   > Set FrontEnd_Idle_first(FrontEnd_Idle_first:bool=true) -> (Output_Get:bool)
   > Set FrontEnd_Idle_second(FrontEnd_Idle_second:bool=false) -> (Output_Get:bool)
   Set FrontEnd_Idle_first(execute=@n23.else, FrontEnd_Idle_first=false)
   > Set FrontEnd_Idle_second(FrontEnd_Idle_second=true)
   Set Is MoveBack(execute=@n9.then_5, Is MoveBack:bool=AND Boolean(A=Get bIsBackpedaling() -> (bIsBackpedaling:bool), B=Get bIsMoving2D())) -> (Output_Get:bool)
   Is Valid(exec=@n8.then_2, InputObject=Get PlayerPawn() -> (PlayerPawn:Fort Player Pawn Athena Object Reference))
   > n24 = Branch(Condition=Get bIsInFrontEnd())
   Set WPO Globals Function(
    execute=@n24.else
    Material=Get Mat Instance Dynamic()
    Overlay Material=Get Overlay Mat Instance Dynamic()
    Skeletal Mesh=Get SkelMesh()
    "Local Noise Transform  Location"=Get Local Noise Transform Location() -> (Local Noise Transform Location:Vector)
    Local Noise Offset=Get Local Noise Offset()
   )
   > Set WPO Regional Function!(
    execute
    Material:Material Instance Dynamic Object Reference=Get Mat Instance Dynamic()
    Overlay Material:Material Instance Dynamic Object Reference=Get Overlay Mat Instance Dynamic()
    Skeletal Mesh:Skeletal Mesh Component Object Reference=Get SkelMesh()
    Velocity Transform:Name=foot_r
    Velocity Interp Speed:double=Get Velocity Interp Speed()
    Collision Transform:Name=calf_r
    Collision Transform Axis:EAxis Enum
    Invert Collision Transform Axis:bool=false
    Delta Time X:double=@n7.DeltaTimeX
    Previous Velocity Transform Location:"Vector (by ref)"=Get R1 Velocity Transform Location()
    Previous Velocity Transform Velocity:"Vector (by ref)"=Get R1 Velocity Transform Velocity()
    Region:int=1
   ) -> (then)
   > Set WPO Regional Function(
    Material=Get Mat Instance Dynamic()
    Overlay Material=Get Overlay Mat Instance Dynamic()
    Skeletal Mesh=Get SkelMesh()
    Velocity Transform=foot_l
    Velocity Interp Speed=Get Velocity Interp Speed()
    Collision Transform=calf_l
    Invert Collision Transform Axis=true
    Delta Time X=@n7.DeltaTimeX
    Previous Velocity Transform Location=Get R2 Velocity Transform Location()
    Previous Velocity Transform Velocity=Get R2 Velocity Transform Velocity()
    Region=2
   )
   > Set WPO Regional Function(
    Material=Get Mat Instance Dynamic()
    Overlay Material=Get Overlay Mat Instance Dynamic()
    Skeletal Mesh=Get SkelMesh()
    Velocity Transform=pelvis
    Velocity Interp Speed=Get Velocity Interp Speed()
    Collision Transform=pelvis
    Invert Collision Transform Axis=true
    Delta Time X=@n7.DeltaTimeX
    Previous Velocity Transform Location=Get R3 Velocity Transform Location() -> (R3 Velocity Transform Location:Vector)
    Previous Velocity Transform Velocity=Get R3 Velocity Transform Velocity() -> (R3 Velocity Transform Velocity:Vector)
    Region=2
   )
   > n25 = Branch(Condition=Get bIsSkydiveFloating())
   > n26 = Set WPO Pho Wind Intensity(WPO Pho Wind Intensity=1)
   SetScalarParameterValue(
    execute=[@n27.then, @n26.then]
    self=[Get Mat Instance Dynamic(), Get Overlay Mat Instance Dynamic()]
    ParameterName=ClothSim_WindStrength
    Value=Get WPO Pho Wind Intensity()
   )
   n27 = Set WPO Pho Wind Intensity(execute=@n25.else, WPO Pho Wind Intensity=0)
   Branch(execute=@n8.then_3, Condition=Get bIsInFrontEnd())
   > Set FECustomIdleBlend(FECustomIdleBlend:float=GetCurveValue(self=GetOwnerFrontEndAnimInstance(), CurveName=Frontend_Idle_Bypass)) -> (Output_Get:float)
   n28 = Gravity Override State Function DEPRECATED(
    execute=@n8.then_0
    CharacterPartAnimInstance=Self-Reference()
    GameTimeInSeconds=GetGameTimeInSeconds()
    GravityOverrideParams=Get Gravity Override Params()
    RBAN Params=Get RBAN Params()
   )
   > Set Gravity Override(Gravity Override=@n28)
   n29 = Event OnWeaponChanged()
   n30 = Get Dynamic Material Instance(execute=[@n29.then, @n34.OutputPin], SkeletalMeshComponent=GetPartSkeletalMeshComponent())
   n31 = Get CurrentWeapon() -> (CurrentWeapon:Fort Weapon Object Reference)
   n32 = Branch(
    execute=[@n30.then, @n33.else]
    Condition=OR Boolean(
     A=WeaponDataHasAnyMatchingGameplayTags!(
      self:Fort Weapon Object Reference=@n31
      TagContainer:Gameplay Tag Container Structure=(GameplayTags=((TagName="Cosmetics.Pickaxe.Backspin")))
     ) -> (ReturnValue:bool)
     B=Get bIsSkyDiving()
    )
   )
   > SetScalarParameterValue(self=@n30, ParameterName=Hide Element 02, Value=1)
   SetScalarParameterValue(execute=@n32.else, self=@n30, ParameterName=Hide Element 02, Value=0)
   n33 = Branch(Condition=Get bIsSkyDiving())
   > SetScalarParameterValue(self=@n30, ParameterName=Hide Element 01, Value=1)
   n34 = Reroute Node()
   WeaponDataHasAnyMatchingGameplayTags(self=@n31, TagContainer=(GameplayTags=((TagName="Weapon.Melee.Impact.Pickaxe"))))
  ]}
]
variables [
 " Tassle Gravity Override Params" {
  type=GravityOverrideParamsStruct
  default {
   emoteOrMelee {gravityOverride=<0 -50 -15>, windAmplitude=<50 -25 100>, windFrequency=<1.5 10 4>, windOffset=<0 -100 -50>}
   front_End {gravityOverride=<0 -50 -15>, windAmplitude=<50 -25 100>, windFrequency=<1.5 10 4>, windOffset=<0 -100 -50>}
   onGround_Moving {jointName=gravity, gravityOverride=<0 -50 15>, windAmplitude=<50 -25 100>, windFrequency=<1.8 13 6>, windOffset=<0 -100 -50>}
   onGround_Standing {jointName=gravity, gravityOverride=<0 -50 -5>, windAmplitude=<50 -25 100>, windFrequency=<1.8 13 6>, windOffset=<0 -100 -50>}
   skydive_FloatingDown {gravityOverride=<0 -700 150>}
   skydive_GoingUp {gravityOverride=<0 -300 150>}
   tactical_Sprint {jointName=gravity, gravityOverride=<0 -450 15>}
  }
 }
 " Tassle GravityOverride" {type=Vector}
 custom_cape_equipped {type=boolean}
 FECustomIdleBlend {type=number, default=1}
 FrontEnd_Idle_first {type=boolean}
 FrontEnd_Idle_second {type=boolean}
 Gravity Override {type=Vector}
 Gravity Override Params {
  type=GravityOverrideParamsStruct
  default {
   skydive_FloatingDown {gravityOverride=<0 -700 500>, windAmplitude=<200 -500 200>, windFrequency=<1.5 7 4>, windOffset=<0 -600 200>}
   skydive_GoingUp {gravityOverride=<0 -150 400>, windAmplitude=<100 -250 100>, windFrequency=<1.5 7 4>}
   skydive_Parachuting {gravityOverride=<0 -400 -980>, windAmplitude=<100 -250 200>, windFrequency=<0.75 3.5 4>}
  }
 }
 Is Backbling Cape {type=boolean}
 Is JumpGround {type=boolean}
 Is MoveBack {type=boolean}
 Is Wearing Backbling {type=boolean}
 Local Noise Offset {type=Vector}
 Local Noise Transform Location {type=Vector}
 Mat Instance Dynamic {type=MaterialInstanceDynamic}
 Overlay Mat Instance Dynamic {type=MaterialInstanceDynamic}
 PlayerPawn {type=FortPlayerPawnAthena}
 R1 Velocity Transform Location {type=Vector}
 R1 Velocity Transform Velocity {type=Vector}
 R2 Velocity Transform Location {type=Vector}
 R2 Velocity Transform Velocity {type=Vector}
 R3 Velocity Transform Location {type=Vector}
 R3 Velocity Transform Velocity {type=Vector}
 RBAN Params {
  type=RBANParamsStruct
  default {
   dBNO {simSpaceSettings {
     maxAngularAcceleration=1000
     maxAngularVelocity=3
     maxLinearAcceleration=300
     maxLinearVelocity=700
     velocityScaleZ=0
     externalLinearDragV=<0 0.5 0>
    }}
   dBNO_Carried {simSpaceSettings {
     maxAngularAcceleration=1000
     maxAngularVelocity=3
     maxLinearAcceleration=300
     maxLinearVelocity=700
     velocityScaleZ=0
     externalLinearDragV=<0 0.5 0>
    }}
   emoteOrMelee {simSpaceSettings {
     maxAngularAcceleration=1000
     maxAngularVelocity=3
     maxLinearAcceleration=300
     maxLinearVelocity=700
     velocityScaleZ=0
     externalLinearDragV=<0 0.5 0>
    }}
   front_End {simSpaceSettings {
     maxAngularAcceleration=1000
     maxAngularVelocity=3
     maxLinearAcceleration=300
     maxLinearVelocity=700
     velocityScaleZ=0
     externalLinearDragV=<0 0.5 0>
    }}
   onGround_Crouch_Moving {simSpaceSettings {
     maxAngularAcceleration=1000
     maxAngularVelocity=3
     maxLinearAcceleration=300
     maxLinearVelocity=700
     velocityScaleZ=0
     externalLinearDragV=<0 0.5 0>
    }}
   onGround_Falling {simSpaceSettings {
     maxAngularAcceleration=1000
     maxAngularVelocity=3
     maxLinearAcceleration=300
     maxLinearVelocity=700
     velocityScaleZ=0
     externalLinearDragV=<0 0.5 0>
    }}
   onGround_Moving {simSpaceSettings {
     maxAngularAcceleration=1000
     maxAngularVelocity=3
     maxLinearAcceleration=300
     maxLinearVelocity=700
     velocityScaleZ=0
     externalLinearDragV=<0 0.5 0>
    }}
   onGround_Standing {simSpaceSettings {
     maxAngularAcceleration=1000
     maxAngularVelocity=3
     maxLinearAcceleration=300
     maxLinearVelocity=700
     velocityScaleZ=0
     externalLinearDragV=<0 0.5 0>
    }}
   rail_Grinding {simSpaceSettings {
     maxAngularAcceleration=1000
     maxAngularVelocity=3
     maxLinearAcceleration=300
     maxLinearVelocity=700
     velocityScaleZ=0
     worldAlpha=0.1
     externalLinearDragV=<0 0.5 0>
    }}
   riding_Moving {simSpaceSettings {
     maxAngularAcceleration=1000
     maxAngularVelocity=3
     maxLinearAcceleration=300
     maxLinearVelocity=700
     velocityScaleZ=0
     externalLinearDragV=<0 0.5 0>
    }}
   riding_Moving_Targeting {simSpaceSettings {
     maxAngularAcceleration=1000
     maxAngularVelocity=3
     maxLinearAcceleration=300
     maxLinearVelocity=700
     velocityScaleZ=0
     externalLinearDragV=<0 0.5 0>
    }}
   riding_Standing {simSpaceSettings {
     maxAngularAcceleration=1000
     maxAngularVelocity=3
     maxLinearAcceleration=300
     maxLinearVelocity=700
     velocityScaleZ=0
     externalLinearDragV=<0 0.5 0>
    }}
   skydive_DivingDown {simSpaceSettings {
     maxAngularAcceleration=1000
     maxAngularVelocity=3
     maxLinearAcceleration=300
     maxLinearVelocity=700
     velocityScaleZ=0
     externalLinearDragV=<0 0.5 0>
    }}
   skydive_FloatingDown {
    componentLinearAccScale=<0 0.5 0>
    componentLinearVelScale=<0 0.5 0>
    simSpaceSettings {
     maxAngularAcceleration=1000
     maxAngularVelocity=3
     maxLinearAcceleration=300
     maxLinearVelocity=700
     velocityScaleZ=0
     externalLinearDragV=<0 0.5 0>
    }
   }
   skydive_GoingUp {simSpaceSettings {
     maxAngularAcceleration=1000
     maxAngularVelocity=3
     maxLinearAcceleration=300
     maxLinearVelocity=700
     velocityScaleZ=0
     externalLinearDragV=<0 0.5 0>
    }}
   skydive_Parachuting {
    componentLinearAccScale=<0 0.1 0.3>
    componentLinearVelScale=<0 0.1 0.3>
    simSpaceSettings {
     maxAngularAcceleration=1000
     maxAngularVelocity=3
     maxLinearAcceleration=300
     maxLinearVelocity=700
     velocityScaleZ=0
     externalLinearDragV=<0 0.5 0>
    }
   }
   skydive_Surfing {simSpaceSettings {
     maxAngularAcceleration=1000
     maxAngularVelocity=3
     maxLinearAcceleration=300
     maxLinearVelocity=700
     velocityScaleZ=0
     externalLinearDragV=<0 0.5 0>
    }}
   swimming {simSpaceSettings {
     maxAngularAcceleration=1000
     maxAngularVelocity=3
     maxLinearAcceleration=300
     maxLinearVelocity=700
     velocityScaleZ=0
     externalLinearDragV=<0 0.5 0>
    }}
   tactical_Sprint {simSpaceSettings {
     maxAngularAcceleration=1000
     maxAngularVelocity=3
     maxLinearAcceleration=300
     maxLinearVelocity=700
     velocityScaleZ=0
     externalLinearDragV=<0 0.5 0>
    }}
   ziplining {simSpaceSettings {
     maxAngularAcceleration=1000
     maxAngularVelocity=3
     maxLinearAcceleration=300
     maxLinearVelocity=700
     velocityScaleZ=0
     externalLinearDragV=<0 0.5 0>
    }}
  }
 }
 Rotate spine {type=number}
 Scarf Visible {type=boolean}
 SkelMesh {type=SkeletalMeshComponent}
 Tassle Alpha {type=number}
 Tassle Rotator {type=Rotator}
 Velocity Interp Speed {type=number, default=30}
 WPO Pho Wind Intensity {type=number}
]
